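# Physics asset that covers only the dress of the F_MED_ForwardLake_DireLab female player cosmetic body, for simulating its skirt motion.
# Six bone chains (center, left and right, front and back) of capsule bodies hang from driver bodies through angle-limited constraints, with left and right front chains starting at segment 2.
# Unconstrained thigh, calf, foot and pelvis capsules represent the lower body, and a large root box sits at ground level like a floor.
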PhysicsAsset /BRCosmetics/Characters/Player/Female/Medium/Bodies/F_MED_ForwardLake_DireLab/Meshes/F_MED_ForwardLake_DireLab_DressOnlyPhysics
bodies {
 calf_l [{shapeType=Capsule, length=150, radius=9, shapeName=calf_l_capsule, center=<-77 -1 1>, rotation=<rot 90 0 0>}]
 calf_r [{shapeType=Capsule, length=150, radius=9, shapeName=calf_l_capsule, center=<77 1 -1>, rotation=<rot 90 0 0>}]
 dyn_C_backDress_1 [{shapeType=Capsule, length=70, radius=10, shapeName=dyn_L_frontDress_1_capsule, center=<-10 0 9>, rotation=<rot 0 0 90>}]
 dyn_C_backDress_2 [{shapeType=Capsule, length=70, radius=13, shapeName=dyn_L_frontDress_1_capsule, center=<-10 0 11>, rotation=<rot 0 0 90>}]
 dyn_C_backDress_3 [{shapeType=Capsule, length=70, radius=13, shapeName=dyn_L_frontDress_1_capsule, center=<-10 0 11>, rotation=<rot 0 0 90>}]
 dyn_C_backDress_4 [{shapeType=Capsule, length=70, radius=12, shapeName=dyn_L_frontDress_1_capsule, center=<-10 0 12>, rotation=<rot 0 0 90>}]
 dyn_C_backDress_driver [{shapeType=Capsule, length=1, radius=1, shapeName=dyn_L_backDress_driver_capsule}]
 dyn_C_frontDress_1 [{shapeType=Capsule, length=120, radius=8.5, shapeName=dyn_L_frontDress_1_capsule, center=<10 0 11>, rotation=<rot 0 0 90>}]
 dyn_C_frontDress_2 [{shapeType=Capsule, length=120, radius=8, shapeName=dyn_L_frontDress_1_capsule, center=<10 0 9.5>, rotation=<rot 0 0 90>}]
 dyn_C_frontDress_3 [{shapeType=Capsule, length=120, radius=9, shapeName=dyn_L_frontDress_1_capsule, center=<10 0 7>, rotation=<rot 0 0 90>}]
 dyn_C_frontDress_4 [{shapeType=Capsule, length=120, radius=9, shapeName=dyn_L_frontDress_1_capsule, center=<10 0 4.5>, rotation=<rot 0 0 90>}]
 dyn_C_frontDress_driver [{shapeType=Capsule, length=120, radius=9, shapeName=dyn_L_frontDress_1_capsule, center=<10 0 10.5>, rotation=<rot 0 0 90>}]
 dyn_L_backDress_1 [{shapeType=Capsule, length=20, radius=12, shapeName=dyn_L_backDress_2_capsule, center=<10 0 13>, rotation=<rot 0 0 90>}]
 dyn_L_backDress_2 [{shapeType=Capsule, length=20, radius=13, shapeName=dyn_L_backDress_2_capsule, center=<10 0 12>, rotation=<rot 0 0 90>}]
 dyn_L_backDress_3 [{shapeType=Capsule, length=35, radius=14, shapeName=dyn_L_backDress_3_capsule, center=<10 0 13>, rotation=<rot 0 0 90>}]
 dyn_L_backDress_4 [{shapeType=Capsule, length=45, radius=13, shapeName=dyn_L_backDress_4_capsule, center=<14 0 13>, rotation=<rot 0 0 90>}]
 dyn_L_backDress_driver [{shapeType=Capsule, length=1, radius=1, shapeName=dyn_L_backDress_driver_capsule}]
 dyn_L_frontDress_2 [{shapeType=Capsule, length=20, radius=10, shapeName=dyn_L_frontDress_2_capsule, center=<10 0 12>, rotation=<rot 0 0 90>}]
 dyn_L_frontDress_3 [{shapeType=Capsule, length=20, radius=9, shapeName=dyn_L_frontDress_3_capsule, center=<10 0 9>, rotation=<rot 0 0 90>}]
 dyn_L_frontDress_4 [{shapeType=Capsule, length=20, radius=9, shapeName=dyn_L_frontDress_4_capsule, center=<12 0 6>, rotation=<rot 0 0 90>}]
 dyn_L_frontDress_driver [{shapeType=Capsule, length=1, radius=1, shapeName=dyn_L_frontDress_driver_capsule}]
 dyn_R_backDress_1 [{shapeType=Capsule, length=20, radius=12, shapeName=dyn_R_backDress_2_capsule, center=<-10 0 -12>, rotation=<rot 0 0 90>}]
 dyn_R_backDress_2 [{shapeType=Capsule, length=20, radius=13.5, shapeName=dyn_R_backDress_2_capsule, center=<-10 0 -12>, rotation=<rot 0 0 90>}]
 dyn_R_backDress_3 [{shapeType=Capsule, length=35, radius=14, shapeName=dyn_R_backDress_3_capsule, center=<-10 0 -12>, rotation=<rot 0 0 90>}]
 dyn_R_backDress_4 [{shapeType=Capsule, length=45, radius=13, shapeName=dyn_R_backDress_4_capsule, center=<-14 0 -12>, rotation=<rot 0 0 90>}]
 dyn_R_backDress_driver [{shapeType=Capsule, length=1, radius=1, shapeName=dyn_R_backDress_driver_capsule}]
 dyn_R_frontDress_2 [{shapeType=Capsule, length=20, radius=10, shapeName=dyn_R_frontDress_2_capsule, center=<-10 0 -12>, rotation=<rot 0 0 90>}]
 dyn_R_frontDress_3 [{shapeType=Capsule, length=20, radius=9, shapeName=dyn_R_frontDress_3_capsule, center=<-10 0 -9>, rotation=<rot 0 0 90>}]
 dyn_R_frontDress_4 [{shapeType=Capsule, length=20, radius=9, shapeName=dyn_R_frontDress_4_capsule, center=<-12 0 -6.5>, rotation=<rot 0 0 90>}]
 dyn_R_frontDress_driver [{shapeType=Capsule, length=1, radius=1, shapeName=dyn_R_frontDress_driver_capsule}]
 foot_l [{shapeType=Capsule, length=10, radius=9, shapeName=foot_l_capsule, center=<-4.7232 5 0>, rotation=<rot 0 0 90>}]
 foot_r [{shapeType=Capsule, length=10, radius=9, shapeName=foot_r_capsule, center=<4.7232 -5 0>, rotation=<rot 0 0 90>}]
 pelvis [{shapeType=Capsule, length=1, radius=1, shapeName=pelvis_capsule}]
 root [{shapeType=Box, extentX=200, extentY=200, extentZ=100, shapeName=root_box, center=<0 0 -48.2435>, rotation=<rot 0 -90 0>}]
 thigh_l [{shapeType=Capsule, length=30, radius=9, shapeName=thigh_l_capsule, center=<-20 0 0>, rotation=<rot 90 0 0>}]
 thigh_r [{shapeType=Capsule, length=30, radius=9, shapeName=thigh_l_capsule, center=<20 0 0>, rotation=<rot 90 0 0>}]
}
constraints [
 {bone1Name=dyn_C_backDress_1, bone2Name=dyn_C_backDress_driver, swing1LimitDegrees=1, swing2LimitDegrees=20, twistLimitDegrees=10}
 {bone1Name=dyn_C_backDress_2, bone2Name=dyn_C_backDress_1, swing1LimitDegrees=1, swing2LimitDegrees=20, twistLimitDegrees=10}
 {bone1Name=dyn_C_backDress_3, bone2Name=dyn_C_backDress_2, swing1LimitDegrees=1, swing2LimitDegrees=20, twistLimitDegrees=10}
 {bone1Name=dyn_C_backDress_4, bone2Name=dyn_C_backDress_3, swing1LimitDegrees=1, swing2LimitDegrees=20, twistLimitDegrees=10}
 {bone1Name=dyn_C_frontDress_1, bone2Name=dyn_C_frontDress_driver, swing1LimitDegrees=5, swing2LimitDegrees=20, twistLimitDegrees=30}
 {bone1Name=dyn_C_frontDress_2, bone2Name=dyn_C_frontDress_1, swing1LimitDegrees=5, swing2LimitDegrees=30, twistLimitDegrees=30}
 {bone1Name=dyn_C_frontDress_3, bone2Name=dyn_C_frontDress_2, swing1LimitDegrees=5, swing2LimitDegrees=30, twistLimitDegrees=30}
 {bone1Name=dyn_C_frontDress_4, bone2Name=dyn_C_frontDress_3, swing1LimitDegrees=5, swing2LimitDegrees=30, twistLimitDegrees=30}
 {bone1Name=dyn_L_backDress_1, bone2Name=dyn_L_backDress_driver, swing1LimitDegrees=5, swing2LimitDegrees=30, twistLimitDegrees=15}
 {bone1Name=dyn_L_backDress_2, bone2Name=dyn_L_backDress_1, swing1LimitDegrees=20, swing2LimitDegrees=20, twistLimitDegrees=20}
 {bone1Name=dyn_L_backDress_3, bone2Name=dyn_L_backDress_2, swing1LimitDegrees=20, swing2LimitDegrees=20, twistLimitDegrees=20}
 {bone1Name=dyn_L_backDress_4, bone2Name=dyn_L_backDress_3, swing1LimitDegrees=20, swing2LimitDegrees=20, twistLimitDegrees=20}
 {bone1Name=dyn_L_frontDress_2, bone2Name=dyn_L_frontDress_driver, swing1LimitDegrees=3, swing2LimitDegrees=20, twistLimitDegrees=5}
 {bone1Name=dyn_L_frontDress_3, bone2Name=dyn_L_frontDress_2, swing1LimitDegrees=15, swing2LimitDegrees=45, twistLimitDegrees=5}
 {bone1Name=dyn_L_frontDress_4, bone2Name=dyn_L_frontDress_3, swing1LimitDegrees=15, swing2LimitDegrees=45, twistLimitDegrees=5}
 {bone1Name=dyn_R_backDress_1, bone2Name=dyn_R_backDress_driver, swing1LimitDegrees=5, swing2LimitDegrees=30, twistLimitDegrees=15}
 {bone1Name=dyn_R_backDress_2, bone2Name=dyn_R_backDress_1, swing1LimitDegrees=20, swing2LimitDegrees=20, twistLimitDegrees=20}
 {bone1Name=dyn_R_backDress_3, bone2Name=dyn_R_backDress_2, swing1LimitDegrees=20, swing2LimitDegrees=20, twistLimitDegrees=20}
 {bone1Name=dyn_R_backDress_4, bone2Name=dyn_R_backDress_3, swing1LimitDegrees=20, swing2LimitDegrees=20, twistLimitDegrees=20}
 {bone1Name=dyn_R_frontDress_2, bone2Name=dyn_R_frontDress_driver, swing1LimitDegrees=3, swing2LimitDegrees=20, twistLimitDegrees=5}
 {bone1Name=dyn_R_frontDress_3, bone2Name=dyn_R_frontDress_2, swing1LimitDegrees=15, swing2LimitDegrees=45, twistLimitDegrees=5}
 {bone1Name=dyn_R_frontDress_4, bone2Name=dyn_R_frontDress_3, swing1LimitDegrees=15, swing2LimitDegrees=45, twistLimitDegrees=5}
]
