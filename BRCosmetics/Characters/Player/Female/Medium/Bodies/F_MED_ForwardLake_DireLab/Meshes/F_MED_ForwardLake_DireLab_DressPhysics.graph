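# Physics asset for the dress and cape secondary motion on the female medium ForwardLake DireLab player body.
# Six dress chains (front and back; left, center, right) hang from driver bones, and three cape chains from MSTR bones, with tight per-link swing and twist limits.
# Free-angle constraints tie both side cape chains to the center at links 02 and 07.
# Capsules on spine, arms and legs, plus a ground-level root box, provide collision shapes.
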PhysicsAsset /BRCosmetics/Characters/Player/Female/Medium/Bodies/F_MED_ForwardLake_DireLab/Meshes/F_MED_ForwardLake_DireLab_DressPhysics
bodies {
 calf_l [{shapeType=Capsule, length=150, radius=9, shapeName=calf_l_capsule, center=<-77 -1 1>, rotation=<rot 90 0 0>}]
 calf_r [{shapeType=Capsule, length=150, radius=9, shapeName=calf_l_capsule, center=<77 1 -1>, rotation=<rot 90 0 0>}]
 dyn_C_backDress_1 [{shapeType=Capsule, length=70, radius=10, shapeName=dyn_L_frontDress_1_capsule, center=<-10 0 9>, rotation=<rot 0 0 90>}]
 dyn_C_backDress_2 [{shapeType=Capsule, length=70, radius=13, shapeName=dyn_L_frontDress_1_capsule, center=<-10 0 11>, rotation=<rot 0 0 90>}]
 dyn_C_backDress_3 [{shapeType=Capsule, length=70, radius=13, shapeName=dyn_L_frontDress_1_capsule, center=<-10 0 11>, rotation=<rot 0 0 90>}]
 dyn_C_backDress_4 [{shapeType=Capsule, length=70, radius=12, shapeName=dyn_L_frontDress_1_capsule, center=<-10 0 12>, rotation=<rot 0 0 90>}]
 dyn_C_backDress_driver [{shapeType=Capsule, length=1, radius=1, shapeName=dyn_L_backDress_driver_capsule}]
 dyn_C_frontDress_1 [{shapeType=Capsule, length=120, radius=8.5, shapeName=dyn_L_frontDress_1_capsule, center=<10 0 11>, rotation=<rot 0 0 90>}]
 dyn_C_frontDress_2 [{shapeType=Capsule, length=120, radius=8, shapeName=dyn_L_frontDress_1_capsule, center=<10 0 9.5>, rotation=<rot 0 0 90>}]
 dyn_C_frontDress_3 [{shapeType=Capsule, length=120, radius=9, shapeName=dyn_L_frontDress_1_capsule, center=<10 0 7>, rotation=<rot 0 0 90>}]
 dyn_C_frontDress_4 [{shapeType=Capsule, length=120, radius=9, shapeName=dyn_L_frontDress_1_capsule, center=<10 0 4.5>, rotation=<rot 0 0 90>}]
 dyn_C_frontDress_driver [{shapeType=Capsule, length=120, radius=9, shapeName=dyn_L_frontDress_1_capsule, center=<10 0 10.5>, rotation=<rot 0 0 90>}]
 dyn_L_backDress_1 [{shapeType=Capsule, length=20, radius=12, shapeName=dyn_L_backDress_2_capsule, center=<10 0 13>, rotation=<rot 0 0 90>}]
 dyn_L_backDress_2 [{shapeType=Capsule, length=20, radius=13, shapeName=dyn_L_backDress_2_capsule, center=<10 0 12>, rotation=<rot 0 0 90>}]
 dyn_L_backDress_3 [{shapeType=Capsule, length=35, radius=14, shapeName=dyn_L_backDress_3_capsule, center=<10 0 13>, rotation=<rot 0 0 90>}]
 dyn_L_backDress_4 [{shapeType=Capsule, length=45, radius=13, shapeName=dyn_L_backDress_4_capsule, center=<14 0 13>, rotation=<rot 0 0 90>}]
 dyn_L_backDress_driver [{shapeType=Capsule, length=1, radius=1, shapeName=dyn_L_backDress_driver_capsule}]
 dyn_L_frontDress_2 [{shapeType=Capsule, length=20, radius=10, shapeName=dyn_L_frontDress_2_capsule, center=<10 0 12>, rotation=<rot 0 0 90>}]
 dyn_L_frontDress_3 [{shapeType=Capsule, length=20, radius=9, shapeName=dyn_L_frontDress_3_capsule, center=<10 0 9>, rotation=<rot 0 0 90>}]
 dyn_L_frontDress_4 [{shapeType=Capsule, length=20, radius=9, shapeName=dyn_L_frontDress_4_capsule, center=<12 0 6>, rotation=<rot 0 0 90>}]
 dyn_L_frontDress_driver [{shapeType=Capsule, length=1, radius=1, shapeName=dyn_L_frontDress_driver_capsule}]
 dyn_R_backDress_1 [{shapeType=Capsule, length=20, radius=12, shapeName=dyn_R_backDress_2_capsule, center=<-10 0 -12>, rotation=<rot 0 0 90>}]
 dyn_R_backDress_2 [{shapeType=Capsule, length=20, radius=13.5, shapeName=dyn_R_backDress_2_capsule, center=<-10 0 -12>, rotation=<rot 0 0 90>}]
 dyn_R_backDress_3 [{shapeType=Capsule, length=35, radius=14, shapeName=dyn_R_backDress_3_capsule, center=<-10 0 -12>, rotation=<rot 0 0 90>}]
 dyn_R_backDress_4 [{shapeType=Capsule, length=45, radius=13, shapeName=dyn_R_backDress_4_capsule, center=<-14 0 -12>, rotation=<rot 0 0 90>}]
 dyn_R_backDress_driver [{shapeType=Capsule, length=1, radius=1, shapeName=dyn_R_backDress_driver_capsule}]
 dyn_R_frontDress_2 [{shapeType=Capsule, length=20, radius=10, shapeName=dyn_R_frontDress_2_capsule, center=<-10 0 -12>, rotation=<rot 0 0 90>}]
 dyn_R_frontDress_3 [{shapeType=Capsule, length=20, radius=9, shapeName=dyn_R_frontDress_3_capsule, center=<-10 0 -9>, rotation=<rot 0 0 90>}]
 dyn_R_frontDress_4 [{shapeType=Capsule, length=20, radius=9, shapeName=dyn_R_frontDress_4_capsule, center=<-12 0 -6.5>, rotation=<rot 0 0 90>}]
 dyn_R_frontDress_driver [{shapeType=Capsule, length=1, radius=1, shapeName=dyn_R_frontDress_driver_capsule}]
 dyn_cape_01 [{shapeType=Capsule, length=20, radius=7, shapeName=dyn_cape_01_capsule, center=<7 0 25>, rotation=<rot 90 0 0>}]
 dyn_cape_02 [{shapeType=Capsule, length=20, radius=7, shapeName=dyn_cape_02_capsule, center=<7 0 25>, rotation=<rot 90 0 0>}]
 dyn_cape_03 [{shapeType=Capsule, length=20, radius=7.5, shapeName=dyn_cape_03_capsule, center=<7 0 24>, rotation=<rot 90 0 0>}]
 dyn_cape_04 [{shapeType=Capsule, length=20, radius=8, shapeName=dyn_cape_04_capsule, center=<10 0 22>, rotation=<rot 90 0 0>}]
 dyn_cape_05 [{shapeType=Capsule, length=20, radius=14.0177, shapeName=dyn_cape_05_capsule, center=<11.6595 0 28>, rotation=<rot 90 0 0>}]
 dyn_cape_06 [{shapeType=Capsule, length=25, radius=14.8785, shapeName=dyn_cape_06_capsule, center=<12.6595 0 29>, rotation=<rot 0 0 90>}]
 dyn_cape_07 [{shapeType=Capsule, length=35, radius=14.8785, shapeName=dyn_cape_07_capsule, center=<17.6595 0 27>, rotation=<rot 0 0 90>}]
 dyn_cape_MSTR [{shapeType=Capsule, length=0.505, radius=0.505, shapeName=dyn_cape_MSTR_capsule, rotation=<rot 0 -90 0>}]
 dyn_cape_l_01 [{shapeType=Capsule, length=20, radius=7, shapeName=dyn_cape_l_01_capsule, center=<7 0 26>, rotation=<rot 90 0 0>}]
 dyn_cape_l_02 [{shapeType=Capsule, length=20, radius=7, shapeName=dyn_cape_l_02_capsule, center=<7 0 25>, rotation=<rot 90 0 0>}]
 dyn_cape_l_03 [{shapeType=Capsule, length=20, radius=7.5, shapeName=dyn_cape_l_03_capsule, center=<7 0 25>, rotation=<rot 90 0 0>}]
 dyn_cape_l_04 [{shapeType=Capsule, length=20, radius=8, shapeName=dyn_cape_l_04_capsule, center=<10 0 23>, rotation=<rot 90 0 0>}]
 dyn_cape_l_05 [{shapeType=Capsule, length=20, radius=14.0177, shapeName=dyn_cape_l_05_capsule, center=<11.6585 -0.0574 29>, rotation=<rot 90 0 0>}]
 dyn_cape_l_06 [{shapeType=Capsule, length=25, radius=14.8785, shapeName=dyn_cape_l_06_capsule, center=<12.6585 -0.0574 30>, rotation=<rot 0 0 90>}]
 dyn_cape_l_07 [{shapeType=Capsule, length=35, radius=14.8785, shapeName=dyn_cape_l_07_capsule, center=<17.6585 -0.0574 28>, rotation=<rot 0 0 90>}]
 dyn_cape_l_MSTR [{shapeType=Capsule, length=0.505, radius=0.505, shapeName=dyn_cape_l_MSTR_capsule, rotation=<rot 0 -90 0>}]
 dyn_cape_r_01 [{shapeType=Capsule, length=20, radius=7, shapeName=dyn_cape_l_01_capsule, center=<7 0 25>, rotation=<rot 90 0 0>}]
 dyn_cape_r_02 [{shapeType=Capsule, length=20, radius=7, shapeName=dyn_cape_l_02_capsule, center=<7 0 26>, rotation=<rot 90 0 0>}]
 dyn_cape_r_03 [{shapeType=Capsule, length=20, radius=7.5, shapeName=dyn_cape_l_03_capsule, center=<7 0 25>, rotation=<rot 90 0 0>}]
 dyn_cape_r_04 [{shapeType=Capsule, length=20, radius=8, shapeName=dyn_cape_l_04_capsule, center=<10 0 23.5>, rotation=<rot 90 0 0>}]
 dyn_cape_r_05 [{shapeType=Capsule, length=20, radius=14.0177, shapeName=dyn_cape_l_05_capsule, center=<11.6585 0.0574 29>, rotation=<rot 90 0 0>}]
 dyn_cape_r_06 [{shapeType=Capsule, length=25, radius=14.8785, shapeName=dyn_cape_l_06_capsule, center=<12.6585 0.0574 30>, rotation=<rot 0 0 90>}]
 dyn_cape_r_07 [{shapeType=Capsule, length=35, radius=14.8785, shapeName=dyn_cape_l_07_capsule, center=<17.6585 0.0574 28>, rotation=<rot 0 0 90>}]
 dyn_cape_r_MSTR [{shapeType=Capsule, length=0.505, radius=0.505, shapeName=dyn_cape_l_MSTR_capsule, rotation=<rot 0 -90 0>}]
 foot_l [{shapeType=Capsule, length=10, radius=9, shapeName=foot_l_capsule, center=<-4.7232 5 0>, rotation=<rot 0 0 90>}]
 foot_r [{shapeType=Capsule, length=10, radius=9, shapeName=foot_r_capsule, center=<4.7232 -5 0>, rotation=<rot 0 0 90>}]
 lowerarm_l [{shapeType=Capsule, length=12, radius=20, shapeName=lowerarm_r_capsule, center=<13 0 0>, rotation=<rot 90 0 0>}]
 lowerarm_r [{shapeType=Capsule, length=12, radius=20, shapeName=lowerarm_r_capsule, center=<-13 0 0>, rotation=<rot 90 0 0>}]
 pelvis [{shapeType=Capsule, length=1, radius=1, shapeName=pelvis_capsule}]
 root [{shapeType=Box, extentX=200, extentY=200, extentZ=100, shapeName=root_box, center=<0 0 -48.2435>, rotation=<rot 0 -90 0>}]
 spine_02 [{shapeType=Capsule, length=100, radius=28.5, shapeName=spine_03_capsule, center=<0 -4.5505 0>}]
 spine_03 [{shapeType=Capsule, length=100, radius=28.5, shapeName=spine_03_capsule, center=<0 -2.3604 0>}]
 spine_05 [{shapeType=Capsule, length=100, radius=28, shapeName=spine_05_capsule, center=<-0.1389 2.718 0>}]
 thigh_l [{shapeType=Capsule, length=30, radius=9, shapeName=thigh_l_capsule, center=<-20 0 0>, rotation=<rot 90 0 0>}]
 thigh_r [{shapeType=Capsule, length=30, radius=9, shapeName=thigh_l_capsule, center=<20 0 0>, rotation=<rot 90 0 0>}]
 upperarm_l [{shapeType=Capsule, length=15, radius=15, shapeName=upperarm_r_capsule, center=<15 0 0>, rotation=<rot 90 0 0>}]
 upperarm_r [{shapeType=Capsule, length=15, radius=15, shapeName=upperarm_r_capsule, center=<-15.0045 0 0>, rotation=<rot 90 0 0>}]
}
constraints [
 {bone1Name=dyn_L_frontDress_2, bone2Name=dyn_L_frontDress_driver, swing1LimitDegrees=3, swing2LimitDegrees=20, twistLimitDegrees=5}
 {bone1Name=dyn_L_frontDress_3, bone2Name=dyn_L_frontDress_2, swing1LimitDegrees=15, swing2LimitDegrees=45, twistLimitDegrees=5}
 {bone1Name=dyn_L_frontDress_4, bone2Name=dyn_L_frontDress_3, swing1LimitDegrees=15, swing2LimitDegrees=45, twistLimitDegrees=5}
 {bone1Name=dyn_R_backDress_2, bone2Name=dyn_R_backDress_1, swing1LimitDegrees=20, swing2LimitDegrees=20, twistLimitDegrees=20}
 {bone1Name=dyn_R_backDress_3, bone2Name=dyn_R_backDress_2, swing1LimitDegrees=20, swing2LimitDegrees=20, twistLimitDegrees=20}
 {bone1Name=dyn_R_backDress_4, bone2Name=dyn_R_backDress_3, swing1LimitDegrees=20, swing2LimitDegrees=20, twistLimitDegrees=20}
 {bone1Name=dyn_R_frontDress_2, bone2Name=dyn_R_frontDress_driver, swing1LimitDegrees=3, swing2LimitDegrees=20, twistLimitDegrees=5}
 {bone1Name=dyn_R_frontDress_3, bone2Name=dyn_R_frontDress_2, swing1LimitDegrees=15, swing2LimitDegrees=45, twistLimitDegrees=5}
 {bone1Name=dyn_R_frontDress_4, bone2Name=dyn_R_frontDress_3, swing1LimitDegrees=15, swing2LimitDegrees=45, twistLimitDegrees=5}
 {bone1Name=dyn_R_backDress_1, bone2Name=dyn_R_backDress_driver, swing1LimitDegrees=5, swing2LimitDegrees=30, twistLimitDegrees=15}
 {bone1Name=dyn_cape_l_01, bone2Name=dyn_cape_l_MSTR, swing1LimitDegrees=4, swing2LimitDegrees=30, twistLimitDegrees=5}
 {bone1Name=dyn_cape_l_02, bone2Name=dyn_cape_l_01, swing1LimitDegrees=4, swing2LimitDegrees=30, twistLimitDegrees=5}
 {bone1Name=dyn_cape_l_03, bone2Name=dyn_cape_l_02, swing1LimitDegrees=4, swing2LimitDegrees=30, twistLimitDegrees=5}
 {bone1Name=dyn_cape_l_04, bone2Name=dyn_cape_l_03, swing1LimitDegrees=4, swing2LimitDegrees=30, twistLimitDegrees=5}
 {bone1Name=dyn_cape_l_05, bone2Name=dyn_cape_l_04, swing1LimitDegrees=3, swing2LimitDegrees=30, twistLimitDegrees=5}
 {bone1Name=dyn_cape_l_06, bone2Name=dyn_cape_l_05, swing1LimitDegrees=3, swing2LimitDegrees=30, twistLimitDegrees=5}
 {bone1Name=dyn_cape_l_07, bone2Name=dyn_cape_l_06, swing1LimitDegrees=3, swing2LimitDegrees=30, twistLimitDegrees=5}
 {bone1Name=dyn_cape_01, bone2Name=dyn_cape_MSTR, swing1LimitDegrees=4, swing2LimitDegrees=30, twistLimitDegrees=5}
 {bone1Name=dyn_cape_02, bone2Name=dyn_cape_01, swing1LimitDegrees=4, swing2LimitDegrees=30, twistLimitDegrees=5}
 {bone1Name=dyn_cape_03, bone2Name=dyn_cape_02, swing1LimitDegrees=4, swing2LimitDegrees=30, twistLimitDegrees=5}
 {bone1Name=dyn_cape_04, bone2Name=dyn_cape_03, swing1LimitDegrees=4, swing2LimitDegrees=30, twistLimitDegrees=5}
 {bone1Name=dyn_cape_05, bone2Name=dyn_cape_04, swing1LimitDegrees=3, swing2LimitDegrees=30, twistLimitDegrees=5}
 {bone1Name=dyn_cape_06, bone2Name=dyn_cape_05, swing1LimitDegrees=3, swing2LimitDegrees=30, twistLimitDegrees=5}
 {bone1Name=dyn_cape_07, bone2Name=dyn_cape_06, swing1LimitDegrees=3, swing2LimitDegrees=30, twistLimitDegrees=5}
 {bone1Name=dyn_C_backDress_1, bone2Name=dyn_C_backDress_driver, swing1LimitDegrees=1, swing2LimitDegrees=20, twistLimitDegrees=10}
 {bone1Name=dyn_C_backDress_2, bone2Name=dyn_C_backDress_1, swing1LimitDegrees=1, swing2LimitDegrees=20, twistLimitDegrees=10}
 {bone1Name=dyn_C_backDress_3, bone2Name=dyn_C_backDress_2, swing1LimitDegrees=1, swing2LimitDegrees=20, twistLimitDegrees=10}
 {bone1Name=dyn_C_backDress_4, bone2Name=dyn_C_backDress_3, swing1LimitDegrees=1, swing2LimitDegrees=20, twistLimitDegrees=10}
 {bone1Name=dyn_C_frontDress_1, bone2Name=dyn_C_frontDress_driver, swing1LimitDegrees=5, swing2LimitDegrees=20, twistLimitDegrees=30}
 {bone1Name=dyn_C_frontDress_2, bone2Name=dyn_C_frontDress_1, swing1LimitDegrees=5, swing2LimitDegrees=30, twistLimitDegrees=30}
 {bone1Name=dyn_C_frontDress_3, bone2Name=dyn_C_frontDress_2, swing1LimitDegrees=5, swing2LimitDegrees=30, twistLimitDegrees=30}
 {bone1Name=dyn_C_frontDress_4, bone2Name=dyn_C_frontDress_3, swing1LimitDegrees=5, swing2LimitDegrees=30, twistLimitDegrees=30}
 {
  bone1Name=dyn_cape_l_07
  bone2Name=dyn_cape_07
  swing1LimitDegrees=45
  swing1Motion=Free
  swing2LimitDegrees=45
  swing2Motion=Free
  twistLimitDegrees=45
  twistMotion=Free
 }
 {
  bone1Name=dyn_cape_l_02
  bone2Name=dyn_cape_02
  swing1LimitDegrees=45
  swing1Motion=Free
  swing2LimitDegrees=45
  swing2Motion=Free
  twistLimitDegrees=45
  twistMotion=Free
 }
 {bone1Name=dyn_cape_r_01, bone2Name=dyn_cape_r_MSTR, swing1LimitDegrees=4, swing2LimitDegrees=30, twistLimitDegrees=5}
 {bone1Name=dyn_cape_r_02, bone2Name=dyn_cape_r_01, swing1LimitDegrees=4, swing2LimitDegrees=30, twistLimitDegrees=5}
 {bone1Name=dyn_cape_r_03, bone2Name=dyn_cape_r_02, swing1LimitDegrees=4, swing2LimitDegrees=30, twistLimitDegrees=5}
 {bone1Name=dyn_cape_r_04, bone2Name=dyn_cape_r_03, swing1LimitDegrees=4, swing2LimitDegrees=30, twistLimitDegrees=5}
 {bone1Name=dyn_cape_r_05, bone2Name=dyn_cape_r_04, swing1LimitDegrees=3, swing2LimitDegrees=30, twistLimitDegrees=5}
 {bone1Name=dyn_cape_r_06, bone2Name=dyn_cape_r_05, swing1LimitDegrees=3, swing2LimitDegrees=30, twistLimitDegrees=5}
 {bone1Name=dyn_cape_r_07, bone2Name=dyn_cape_r_06, swing1LimitDegrees=3, swing2LimitDegrees=30, twistLimitDegrees=5}
 {
  bone1Name=dyn_cape_r_02
  bone2Name=dyn_cape_02
  swing1LimitDegrees=45
  swing1Motion=Free
  swing2LimitDegrees=45
  swing2Motion=Free
  twistLimitDegrees=45
  twistMotion=Free
 }
 {
  bone1Name=dyn_cape_r_07
  bone2Name=dyn_cape_07
  swing1LimitDegrees=45
  swing1Motion=Free
  swing2LimitDegrees=45
  swing2Motion=Free
  twistLimitDegrees=45
  twistMotion=Free
 }
 {bone1Name=dyn_L_backDress_1, bone2Name=dyn_L_backDress_driver, swing1LimitDegrees=5, swing2LimitDegrees=30, twistLimitDegrees=15}
 {bone1Name=dyn_L_backDress_2, bone2Name=dyn_L_backDress_1, swing1LimitDegrees=20, swing2LimitDegrees=20, twistLimitDegrees=20}
 {bone1Name=dyn_L_backDress_3, bone2Name=dyn_L_backDress_2, swing1LimitDegrees=20, swing2LimitDegrees=20, twistLimitDegrees=20}
 {bone1Name=dyn_L_backDress_4, bone2Name=dyn_L_backDress_3, swing1LimitDegrees=20, swing2LimitDegrees=20, twistLimitDegrees=20}
]
