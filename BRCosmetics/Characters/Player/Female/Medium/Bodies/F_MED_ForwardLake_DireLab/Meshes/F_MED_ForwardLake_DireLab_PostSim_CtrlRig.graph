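# Post-sim Control Rig for the F_MED_ForwardLake_DireLab female character body, applying corrective offsets to dress, cape and thigh-correction bones.
# Dress and thigh angles are remapped into weights for additive bone transforms, while parent constraints blend center-front dress bones between left and right.
# State flags such as crouch, jump, knee slide, skydive, emote, frontend and backbling drive further pose fixes, many smoothed by interpolation.
# Leg-across detection and Kalman filter nodes are disconnected leftovers.
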ControlRigBlueprint /BRCosmetics/Characters/Player/Female/Medium/Bodies/F_MED_ForwardLake_DireLab/Meshes/F_MED_ForwardLake_DireLab_PostSim_CtrlRig
graphs [
 Maximum_1_ContainedGraph {nodes [
   n1 = RigVMFunctionEntryNode()
   Maximum = &RigVMFunction_MathDoubleMax(A=@n1.A, B=@n1.B)
   RigVMFunctionReturnNode(Result=&RigVMFunction_MathDoubleMax(B=@n1.D, A=&RigVMFunction_MathDoubleMax(B=@n1.C, A=@Maximum.Result).Result).Result)
  ]}
 RigVMFunctionLibrary {links [], nodes []}
 RigVMFunction_Sequence_1_ContainedGraph {nodes [
   n1 = RigVMFunctionEntryNode()
   RigVMFunction_Sequence = &RigVMFunction_Sequence(ExecuteContext=@n1.ExecuteContext)
   n2 = &RigVMFunction_Sequence(ExecuteContext=@RigVMFunction_Sequence.B)
   RigVMFunctionReturnNode(A=@RigVMFunction_Sequence.A, B=@n2.A, C=@n2.B)
  ]}
 RigVMModel {nodes [
   RigUnit_BeginExecution = &RigUnit_BeginExecution()
   RigVMFunction_Sequence = RigVMAggregateNode(ExecuteContext=@RigUnit_BeginExecution.ExecutePin)
   RigUnit_GetTransform = &RigUnit_GetTransform(Item=(Type=Bone,Name="thigh_l"), Space=LocalSpace, bInitial=False)
   RigVMVariableNode(Value=@RigUnit_GetTransform.Transform.Rotation, ExecuteContext=@RigVMFunction_Sequence.A) {variable=previous L thigh quat}
   > RigVMVariableNode(Value=&RigUnit_GetTransform(Item=(Type=Bone,Name="thigh_r"), Space=LocalSpace, bInitial=False).Transform.Rotation) {variable=previous R thigh quat}
   RigVMFunction_MathQuaternionToEuler = &RigVMFunction_MathQuaternionToEuler(
    RotationOrder=ZYX
    Value=&RigUnit_GetRelativeTransformForItem(Child=(Type=Bone,Name="dyn_L_frontDress_2"), Parent=(Type=Bone,Name="pelvis"), bChildInitial=False, bParentInitial=False).RelativeTransform.Rotation
   )
   n1 = &RigUnit_GetRelativeTransformForItem(Child=(Type=Bone,Name="dyn_R_frontDress_2"), Parent=(Type=Bone,Name="pelvis"), bChildInitial=False, bParentInitial=False)
   n2 = &RigUnit_GetRelativeTransformForItem(Child=(Type=Bone,Name="dyn_L_backDress_2"), Parent=(Type=Bone,Name="pelvis"), bChildInitial=False, bParentInitial=False)
   n3 = &RigUnit_GetRelativeTransformForItem(Child=(Type=Bone,Name="dyn_R_backDress_2"), Parent=(Type=Bone,Name="pelvis"), bChildInitial=False, bParentInitial=False)
   Maximum = RigVMAggregateNode(
    A=&RigVMFunction_MathDoubleRemap(SourceMaximum=70, SourceMinimum=30, TargetMaximum=1, TargetMinimum=0, Value=@RigVMFunction_MathQuaternionToEuler.Result.Y).Result
    B=&RigVMFunction_MathDoubleRemap(
     SourceMaximum=-70
     SourceMinimum=-30
     TargetMaximum=1
     TargetMinimum=0
     Value=&RigVMFunction_MathQuaternionToEuler(RotationOrder=ZYX, Value=@n1.RelativeTransform.Rotation).Result.Y
    ).Result
    C=&RigVMFunction_MathDoubleRemap(
     SourceMaximum=70
     SourceMinimum=30
     TargetMaximum=1
     TargetMinimum=0
     Value=&RigVMFunction_MathQuaternionToEuler(RotationOrder=ZYX, Value=@n2.RelativeTransform.Rotation).Result.Y
    ).Result
    D=&RigVMFunction_MathDoubleRemap(
     SourceMaximum=-70
     SourceMinimum=-30
     TargetMaximum=1
     TargetMinimum=0
     Value=&RigVMFunction_MathQuaternionToEuler(RotationOrder=ZYX, Value=@n3.RelativeTransform.Rotation).Result.Y
    ).Result
   )
   n4 = &RigVMFunction_AlphaInterp(InterpSpeedDecreasing=30, InterpSpeedIncreasing=30, Scale=1, bClampResult=False, Value=@Maximum.Result)
   &RigUnit_ModifyTransforms(
    ItemToModify="((Item=(Type=Bone,Name=\"dyn_L_thighCorrection\"),Transform=(Rotation=(X=0,Y=0,Z=0,W=1),Translation=(X=7,Y=0,Z=0),Scale3D=(X=0.7,Y=0.9,Z=1))),(Item=(Type=Bone,Name=\"dyn_R_thighCorrection\"),Transform=(Rotation=(X=0,Y=0,Z=0,W=1),Translation=(X=-7,Y=0,Z=0),Scal ...(truncated)"
    WeightMinimum=0
    Weight=@n4.Result
    ExecutePin=@RigVMFunction_Sequence.B
   )
   n5 = RigVMVariableNode() {variable=previous L thigh quat}
   n6 = &RigVMFunction_MathDoubleRemap(
    SourceMaximum=-30
    SourceMinimum=0
    TargetMaximum=1
    TargetMinimum=0
    Value=&RigVMFunction_MathQuaternionToEuler(RotationOrder=ZXY, Value=@n5.Value).Result.Y
   )
   &RigUnit_ModifyTransforms(
    ItemToModify="((Item=(Type=Bone,Name=\"dyn_L_frontDress_driver\"),Transform=(Rotation=(X=0,Y=-0.1305,Z=0,W=0.9914),Translation=(X=0,Y=0,Z=0),Scale3D=(X=1,Y=1,Z=1))),(Item=(Type=Bone,Name=\"dyn_L_backDress_driver\"),Transform=(Rotation=(X=0,Y=0,Z=0.0436,W=0.999),Translation= ...(truncated)"
    WeightMinimum=0
    Weight=@n6.Result
    ExecutePin=@RigVMFunction_Sequence.C
   )
   n7 = &RigUnit_GetRelativeTransformForItem(Child=(Type=Bone,Name="thigh_l"), Parent=(Type=Bone,Name="pelvis"), bChildInitial=False, bParentInitial=False)
   > &RigUnit_ModifyTransforms(
    ItemToModify=((Item=(Type=Bone,Name="dyn_L_sideDress_driver"),Transform=(Rotation=(X=0,Y=-0.1305,Z=0,W=0.9914),Translation=(X=0,Y=0,Z=0),Scale3D=(X=1,Y=1,Z=1))))
    WeightMinimum=0
    Weight=&RigVMFunction_MathDoubleRemap(
     SourceMaximum=-60
     SourceMinimum=-40
     TargetMaximum=1
     TargetMinimum=0
     Value=&RigVMFunction_MathQuaternionToEuler(RotationOrder=ZXY, Value=@n7.RelativeTransform.Rotation).Result.Y
    ).Result
   )
   n8 = RigVMVariableNode() {variable=previous R thigh quat}
   n9 = &RigVMFunction_MathDoubleRemap(
    SourceMaximum=30
    SourceMinimum=0
    TargetMaximum=1
    TargetMinimum=0
    Value=&RigVMFunction_MathQuaternionToEuler(RotationOrder=ZXY, Value=@n8.Value).Result.Y
   )
   > &RigUnit_ModifyTransforms(
    ItemToModify="((Item=(Type=Bone,Name=\"dyn_R_frontDress_driver\"),Transform=(Rotation=(X=0,Y=-0.1305,Z=0,W=0.9914),Translation=(X=0,Y=0,Z=0),Scale3D=(X=1,Y=1,Z=1))),(Item=(Type=Bone,Name=\"dyn_R_backDress_driver\"),Transform=(Rotation=(X=0,Y=0,Z=0.0872,W=0.9962),Translation ...(truncated)"
    WeightMinimum=0
    Weight=@n9.Result
   )
   n10 = &RigUnit_GetRelativeTransformForItem(Child=(Type=Bone,Name="thigh_r"), Parent=(Type=Bone,Name="pelvis"), bChildInitial=False, bParentInitial=False)
   > &RigUnit_ModifyTransforms(
    ItemToModify=((Item=(Type=Bone,Name="dyn_R_sideDress_driver"),Transform=(Rotation=(X=0,Y=-0.1305,Z=0,W=0.9914),Translation=(X=0,Y=0,Z=0),Scale3D=(X=1,Y=1,Z=1))))
    WeightMinimum=0
    Weight=&RigVMFunction_MathDoubleRemap(
     SourceMaximum=60
     SourceMinimum=40
     TargetMaximum=1
     TargetMinimum=0
     Value=&RigVMFunction_MathQuaternionToEuler(RotationOrder=ZXY, Value=@n10.RelativeTransform.Rotation).Result.Y
    ).Result
   )
   n11 = &RigVMFunction_AlphaInterp(
    InterpSpeedDecreasing=10
    InterpSpeedIncreasing=10
    Scale=1
    bClampResult=False
    Value=&RigVMFunction_MathBoolToFloat(Value=RigVMVariableNode() {variable=isCrouchOverall}.Value).Result
   )
   > &RigUnit_ModifyTransforms(
    ItemToModify="((Item=(Type=Bone,Name=\"dyn_L_frontDress_3\"),Transform=(Rotation=(X=-0.0076,Y=0.0868,Z=0.0868,W=0.9924),Translation=(X=4,Y=0,Z=3),Scale3D=(X=1,Y=1,Z=1))),(Item=(Type=Bone,Name=\"dyn_R_frontDress_3\"),Transform=(Rotation=(X=-0.0076,Y=0.0868,Z=0.0868,W=0.9924) ...(truncated)"
    WeightMinimum=0
    Weight=@n11.Result
   )
   RigVMFunction_MathBoolNot = &RigVMFunction_MathBoolNot(Value=RigVMVariableNode() {variable=isCrouchMoving}.Value)
   RigVMFunction_MathBoolAnd = &RigVMFunction_MathBoolAnd(A=@RigVMFunction_MathBoolNot.Result, B=RigVMVariableNode() {variable=isCrouchOverall}.Value)
   n12 = &RigVMFunction_AlphaInterp(
    InterpSpeedDecreasing=10
    InterpSpeedIncreasing=10
    Scale=1
    bClampResult=False
    Value=&RigVMFunction_MathBoolToFloat(Value=@RigVMFunction_MathBoolAnd.Result).Result
   )
   > &RigUnit_ModifyTransforms(
    ItemToModify="((Item=(Type=Bone,Name=\"dyn_C_backDress_1\"),Transform=(Rotation=(X=0,Y=0,Z=0,W=1),Translation=(X=0,Y=0,Z=-5),Scale3D=(X=1,Y=1,Z=1))),(Item=(Type=Bone,Name=\"dyn_C_backDress_2\"),Transform=(Rotation=(X=0,Y=-0.0872,Z=0,W=0.9962),Translation=(X=0,Y=0,Z=-5),Scal ...(truncated)"
    WeightMinimum=0
    Weight=@n12.Result
   )
   > &RigUnit_ModifyTransforms(
    ItemToModify="((Item=(Type=Bone,Name=\"dyn_R_backDress_1\"),Transform=(Rotation=(X=0,Y=0,Z=0.0872,W=0.9962),Translation=(X=0,Y=0,Z=5),Scale3D=(X=1,Y=1,Z=1))),(Item=(Type=Bone,Name=\"dyn_R_backDress_2\"),Transform=(Rotation=(X=0,Y=-0.0872,Z=0,W=0.9962),Translation=(X=0,Y=0,Z ...(truncated)"
    WeightMinimum=0
    Weight=@n12.Result
   )
   > &RigUnit_ModifyTransforms(
    ItemToModify="((Item=(Type=Bone,Name=\"dyn_L_backDress_1\"),Transform=(Rotation=(X=0,Y=0,Z=0.0872,W=0.9962),Translation=(X=0,Y=0,Z=-5),Scale3D=(X=1,Y=1,Z=1))),(Item=(Type=Bone,Name=\"dyn_L_backDress_2\"),Transform=(Rotation=(X=0,Y=-0.0872,Z=0,W=0.9962),Translation=(X=0,Y=0, ...(truncated)"
    WeightMinimum=0
    Weight=@n12.Result
   )
   > &RigUnit_ModifyTransforms(
    ItemToModify="((Item=(Type=Bone,Name=\"dyn_L_frontDress_3\"),Transform=(Rotation=(X=0,Y=0.2588,Z=0,W=0.9659),Translation=(X=0,Y=0,Z=0),Scale3D=(X=1,Y=1,Z=1))),(Item=(Type=Bone,Name=\"dyn_R_frontDress_3\"),Transform=(Rotation=(X=0,Y=0,Z=0,W=1),Translation=(X=0,Y=0,Z=0),Scale ...(truncated)"
    WeightMinimum=0
    Weight=@n12.Result
   )
   > &RigUnit_ParentConstraint(
    AdvancedSettings=(InterpolationType=Average,RotationOrderForFilter=XZY)
    Child=(Type=Bone,Name="dyn_C_frontDress_driver")
    Filter=(TranslationFilter=(bX=false,bY=false,bZ=false),RotationFilter=(bX=True,bY=True,bZ=True),ScaleFilter=(bX=True,bY=True,bZ=True))
    Parents=((Item=(Type=Bone,Name="dyn_L_frontDress_driver"),Weight=0.5),(Item=(Type=Bone,Name="dyn_R_frontDress_driver"),Weight=0.5))
    Weight=0.5
   )
   > &RigUnit_ParentConstraint(
    AdvancedSettings=(InterpolationType=Average,RotationOrderForFilter=XZY)
    Child=(Type=Bone,Name="dyn_C_frontDress_2")
    Filter=(TranslationFilter=(bX=false,bY=false,bZ=false),RotationFilter=(bX=True,bY=True,bZ=True),ScaleFilter=(bX=True,bY=True,bZ=True))
    Parents=((Item=(Type=Bone,Name="dyn_L_frontDress_2"),Weight=0.5),(Item=(Type=Bone,Name="dyn_R_frontDress_2"),Weight=0.5))
    Weight=0.5
   )
   > &RigUnit_ParentConstraint(
    AdvancedSettings=(InterpolationType=Average,RotationOrderForFilter=XZY)
    Child=(Type=Bone,Name="dyn_C_frontDress_3")
    Filter=(TranslationFilter=(bX=false,bY=false,bZ=false),RotationFilter=(bX=True,bY=True,bZ=True),ScaleFilter=(bX=True,bY=True,bZ=True))
    Parents=((Item=(Type=Bone,Name="dyn_L_frontDress_3"),Weight=0.5),(Item=(Type=Bone,Name="dyn_R_frontDress_3"),Weight=0.5))
    Weight=0.5
   )
   n13 = &RigVMFunction_MathBoolToFloat(Value=&RigVMFunction_MathBoolNot(Value=RigVMVariableNode() {variable=isRBANenabled}.Value).Result)
   n14 = RigVMVariableNode() {variable=isCustomIdlePlaying}
   n15 = &RigVMFunction_MathQuaternionToEuler(
    RotationOrder=ZXY
    Value=&RigUnit_GetRelativeTransformForItem(Child=(Type=Bone,Name="thigh_r"), Parent=(Type=Bone,Name="pelvis"), bChildInitial=False, bParentInitial=False).RelativeTransform.Rotation
   )
   > &RigUnit_ModifyTransforms(
    ItemToModify=((Item=(Type=Bone,Name="dyn_R_backDress_driver"),Transform=(Rotation=(X=0,Y=0,Z=0.0872,W=0.9962),Translation=(X=0,Y=-3,Z=-3),Scale3D=(X=1,Y=1,Z=1))))
    WeightMinimum=0
    Weight=&RigVMFunction_MathDoubleSub(
     B=&RigVMFunction_MathDoubleMul(A=@n13.Result, B=@n14.Value).Result
     A=&RigVMFunction_MathDoubleRemap(SourceMaximum=20, SourceMinimum=-5, TargetMaximum=1, TargetMinimum=0, Value=@n15.Result.Y).Result
    ).Result
   )
   n16 = &RigVMFunction_MathBoolToFloat(Value=&RigVMFunction_MathBoolNot(Value=RigVMVariableNode() {variable=isRBANenabled}.Value).Result)
   n17 = RigVMVariableNode() {variable=isCustomIdlePlaying}
   n18 = &RigVMFunction_MathQuaternionToEuler(
    RotationOrder=ZXY
    Value=&RigUnit_GetRelativeTransformForItem(Child=(Type=Bone,Name="thigh_l"), Parent=(Type=Bone,Name="pelvis"), bChildInitial=False, bParentInitial=False).RelativeTransform.Rotation
   )
   > &RigUnit_ModifyTransforms(
    ItemToModify=((Item=(Type=Bone,Name="dyn_L_backDress_driver"),Transform=(Rotation=(X=0,Y=0,Z=0.0872,W=0.9962),Translation=(X=0,Y=3,Z=3),Scale3D=(X=1,Y=1,Z=1))))
    WeightMinimum=0
    Weight=&RigVMFunction_MathDoubleSub(
     B=&RigVMFunction_MathDoubleMul(A=@n16.Result, B=@n17.Value).Result
     A=&RigVMFunction_MathDoubleRemap(SourceMaximum=-20, SourceMinimum=5, TargetMaximum=1, TargetMinimum=0, Value=@n18.Result.Y).Result
    ).Result
   )
   RigVMFunction_MathBoolToFloat = &RigVMFunction_MathBoolToFloat(Value=RigVMVariableNode() {variable=isSkydiveFloating}.Value)
   AlphaInterp = &RigVMFunction_AlphaInterp(InterpSpeedDecreasing=10, InterpSpeedIncreasing=10, Scale=1, bClampResult=False, Value=@RigVMFunction_MathBoolToFloat.Result)
   > &RigUnit_ModifyTransforms(
    ItemToModify="((Item=(Type=Bone,Name=\"dyn_L_backDress_3\"),Transform=(Rotation=(X=0,Y=-0.1736,Z=0,W=0.9848),Translation=(X=0,Y=0,Z=5),Scale3D=(X=1,Y=1,Z=1))),(Item=(Type=Bone,Name=\"dyn_R_backDress_3\"),Transform=(Rotation=(X=0,Y=-0.1736,Z=0,W=0.9848),Translation=(X=0,Y=0, ...(truncated)"
    WeightMinimum=0
    Weight=@AlphaInterp.Result
   )
   n19 = &RigVMFunction_MathQuaternionToEuler(
    RotationOrder=ZXY
    Value=&RigUnit_GetRelativeTransformForItem(Child=(Type=Bone,Name="thigh_r"), Parent=(Type=Bone,Name="pelvis"), bChildInitial=False, bParentInitial=False).RelativeTransform.Rotation
   )
   > RigVMVariableNode(Value=&RigVMFunction_MathDoubleRemap(SourceMaximum=-5, SourceMinimum=10, TargetMaximum=1, TargetMinimum=0, Value=@n19.Result.Y).Result) {variable=R thigh inward alpha}
   Absolute = &RigVMFunction_MathDoubleAbs(Value=@n19.Result.Z)
   Subtract = &RigVMFunction_MathDoubleSub(A=180, B=@Absolute.Result)
   > RigVMVariableNode(Value=&RigVMFunction_MathDoubleRemap(SourceMaximum=40, SourceMinimum=0, TargetMaximum=1, TargetMinimum=0, Value=@Subtract.Result).Result) {variable=R thigh forwardAlpha}
   n20 = &RigVMFunction_MathQuaternionToEuler(
    RotationOrder=ZXY
    Value=&RigUnit_GetRelativeTransformForItem(Child=(Type=Bone,Name="thigh_l"), Parent=(Type=Bone,Name="pelvis"), bChildInitial=False, bParentInitial=False).RelativeTransform.Rotation
   )
   > RigVMVariableNode(Value=&RigVMFunction_MathDoubleRemap(SourceMaximum=5, SourceMinimum=-10, TargetMaximum=1, TargetMinimum=0, Value=@n20.Result.Y).Result) {variable=L thigh inward alpha}
   > RigVMVariableNode(Value=&RigVMFunction_MathDoubleRemap(SourceMaximum=-40, SourceMinimum=0, TargetMaximum=1, TargetMinimum=0, Value=@n20.Result.Z).Result) {variable=L thigh foward alpha}
   Multiply = &RigVMFunction_MathDoubleMul(
    B=&RigVMFunction_MathDoubleMul(A=RigVMVariableNode() {variable=R thigh inward alpha}.Value, B=RigVMVariableNode() {variable=L thigh inward alpha}.Value).Result
    A=&RigVMFunction_MathBoolToFloat(Value=RigVMVariableNode() {variable=isCrouchMoving}.Value).Result
   )
   n21 = &RigVMFunction_AlphaInterp(InterpSpeedDecreasing=15, InterpSpeedIncreasing=15, Scale=1, bClampResult=False, Value=@Multiply.Result)
   > &RigUnit_ModifyTransforms(
    ItemToModify="((Item=(Type=Bone,Name=\"dyn_L_sideDress_driver\"),Transform=(Rotation=(X=0,Y=0,Z=0,W=1),Translation=(X=0,Y=0,Z=0),Scale3D=(X=1,Y=1,Z=1))),(Item=(Type=Bone,Name=\"dyn_R_sideDress_driver\"),Transform=(Rotation=(X=0,Y=0,Z=0,W=1),Translation=(X=0,Y=0,Z=0),Scale3D ...(truncated)"
    WeightMinimum=0
    Weight=@n21.Result
   )
   > &RigUnit_ModifyTransforms(
    ItemToModify="((Item=(Type=Bone,Name=\"dyn_L_frontDress_2\"),Transform=(Rotation=(X=0,Y=-0.1305,Z=0,W=0.9914),Translation=(X=0,Y=0,Z=0),Scale3D=(X=1,Y=1,Z=1))),(Item=(Type=Bone,Name=\"dyn_R_frontDress_2\"),Transform=(Rotation=(X=0,Y=-0.1305,Z=0,W=0.9914),Translation=(X=0,Y= ...(truncated)"
    WeightMinimum=0
    Weight=@n21.Result
   )
   n22 = RigVMVariableNode() {variable=isJumping}
   n23 = &RigVMFunction_AlphaInterp(
    InterpSpeedDecreasing=20
    InterpSpeedIncreasing=20
    Scale=1
    bClampResult=False
    Value=&RigVMFunction_MathDoubleMul(A=&RigVMFunction_MathBoolToFloat(Value=@n22.Value).Result, B=RigVMVariableNode() {variable=R thigh forwardAlpha}.Value).Result
   )
   > &RigUnit_ModifyTransforms(
    ItemToModify=((Item=(Type=Bone,Name="dyn_R_frontDress_2"),Transform=(Rotation=(X=0,Y=0,Z=-0.0872,W=0.9962),Translation=(X=0,Y=0,Z=0),Scale3D=(X=1,Y=1,Z=1))))
    WeightMinimum=0
    Weight=@n23.Result
   )
   n24 = RigVMVariableNode() {variable=isJumping}
   n25 = &RigVMFunction_AlphaInterp(
    InterpSpeedDecreasing=20
    InterpSpeedIncreasing=20
    Scale=1
    bClampResult=False
    Value=&RigVMFunction_MathDoubleMul(A=&RigVMFunction_MathBoolToFloat(Value=@n24.Value).Result, B=RigVMVariableNode() {variable=L thigh foward alpha}.Value).Result
   )
   > &RigUnit_ModifyTransforms(
    ItemToModify=((Item=(Type=Bone,Name="dyn_L_frontDress_2"),Transform=(Rotation=(X=0,Y=0,Z=-0.0872,W=0.9962),Translation=(X=0,Y=0,Z=0),Scale3D=(X=1,Y=1,Z=1))))
    WeightMinimum=0
    Weight=@n25.Result
   )
   n26 = &RigVMFunction_MathBoolAnd(
    A=&RigVMFunction_MathBoolNot(Value=RigVMVariableNode() {variable=isCrouchMoving}.Value).Result
    B=RigVMVariableNode() {variable=isCrouchOverall}.Value
   )
   n27 = RigVMVariableNode() {variable=isWearingBackbling}
   n28 = &RigVMFunction_AlphaInterp(
    InterpSpeedDecreasing=10
    InterpSpeedIncreasing=10
    Scale=1
    bClampResult=False
    Value=&RigVMFunction_MathBoolToFloat(Value=&RigVMFunction_MathBoolAnd(A=@n26.Result, B=@n27.Value).Result).Result
   )
   > &RigUnit_ModifyTransforms(
    ItemToModify="((Item=(Type=Bone,Name=\"dyn_cape_02\"),Transform=(Rotation=(X=0,Y=-0.0436,Z=0,W=0.999),Translation=(X=0,Y=0,Z=0),Scale3D=(X=1,Y=1,Z=1))),(Item=(Type=Bone,Name=\"dyn_cape_l_02\"),Transform=(Rotation=(X=0,Y=-0.0436,Z=0,W=0.999),Translation=(X=0,Y=0,Z=0),Scale3D ...(truncated)"
    WeightMinimum=0
    Weight=@n28.Result
   )
   n29 = RigVMVariableNode() {variable=isRBANenabled}
   > &RigUnit_ModifyTransforms(
    ItemToModify="((Item=(Type=Bone,Name=\"dyn_L_backDress_2\"),Transform=(Rotation=(X=0.0023,Y=0.0261,Z=-0.0871,W=0.9959),Translation=(X=0,Y=0,Z=0),Scale3D=(X=1,Y=1,Z=1))),(Item=(Type=Bone,Name=\"dyn_R_frontDress_3\"),Transform=(Rotation=(X=0,Y=0.0436,Z=0,W=0.999),Translation= ...(truncated)"
    WeightMinimum=0
    Weight=&RigVMFunction_MathDoubleMul(B=&RigVMFunction_MathBoolToFloat(Value=@n29.Value).Result, A=RigVMVariableNode() {variable=isCustomIdlePlaying}.Value).Result
   )
   n30 = RigVMVariableNode() {variable=isFrontend}
   n31 = RigVMVariableNode() {variable=isWearingBackbling}
   > &RigUnit_ModifyTransforms(
    ItemToModify="((Item=(Type=Bone,Name=\"dyn_L_backDress_3\"),Transform=(Rotation=(X=0.0023,Y=0.0261,Z=-0.0871,W=0.9959),Translation=(X=0,Y=0,Z=0),Scale3D=(X=1,Y=1,Z=1))),(Item=(Type=Bone,Name=\"dyn_R_frontDress_3\"),Transform=(Rotation=(X=0,Y=0.0436,Z=0,W=0.999),Translation= ...(truncated)"
    WeightMinimum=0
    Weight=&RigVMFunction_MathBoolToFloat(Value=&RigVMFunction_MathBoolAnd(A=@n30.Value, B=@n31.Value).Result).Result
   )
   > &RigUnit_ModifyTransforms(
    ItemToModify="((Item=(Type=Bone,Name=\"dyn_L_backDress_2\"),Transform=(Rotation=(X=0.0023,Y=0.0261,Z=-0.0871,W=0.9959),Translation=(X=0,Y=0,Z=0),Scale3D=(X=1,Y=1,Z=1))),(Item=(Type=Bone,Name=\"dyn_R_frontDress_3\"),Transform=(Rotation=(X=0,Y=0.0436,Z=0,W=0.999),Translation= ...(truncated)"
    WeightMinimum=0
    Weight=&RigVMFunction_MathBoolToFloat(Value=RigVMVariableNode() {variable=isPlayingEmote}.Value).Result
   )
   n32 = &RigVMFunction_MathBoolNot(Value=RigVMVariableNode() {variable=isPlayingEmote}.Value)
   n33 = RigVMVariableNode() {variable=isFrontend}
   > &RigUnit_ModifyTransforms(
    ItemToModify=((Item=(Type=Bone,Name="dyn_R_backDress_4"),Transform=(Rotation=(X=0.0003,Y=-0.0175,Z=0.0175,W=0.9997),Translation=(X=0,Y=0,Z=0),Scale3D=(X=1,Y=1,Z=1))))
    WeightMinimum=0
    Weight=&RigVMFunction_MathBoolToFloat(Value=&RigVMFunction_MathBoolAnd(B=@n32.Result, A=@n33.Value).Result).Result
   )
   n34 = RigVMVariableNode() {variable=isKneeSliding}
   n35 = RigVMVariableNode() {variable=isRBANenabled}
   n36 = &RigVMFunction_AlphaInterp(
    InterpSpeedDecreasing=10
    InterpSpeedIncreasing=10
    Scale=1
    bClampResult=False
    Value=&RigVMFunction_MathBoolToFloat(Value=&RigVMFunction_MathBoolAnd(B=@n34.Value, A=@n35.Value).Result).Result
   )
   > &RigUnit_ModifyTransforms(
    ItemToModify="((Item=(Type=Bone,Name=\"dyn_L_backDress_1\"),Transform=(Rotation=(X=0,Y=-0.1736,Z=0,W=-0.9848),Translation=(X=-20,Y=0,Z=22),Scale3D=(X=1,Y=1,Z=1))),(Item=(Type=Bone,Name=\"dyn_R_backDress_1\"),Transform=(Rotation=(X=0,Y=-0.1736,Z=0,W=0.9848),Translation=(X=20 ...(truncated)"
    WeightMinimum=0
    Weight=@n36.Result
   )
   > &RigUnit_ModifyTransforms(
    ItemToModify="((Item=(Type=Bone,Name=\"dyn_L_backDress_2\"),Transform=(Rotation=(X=0,Y=0.0872,Z=0,W=0.9962),Translation=(X=2,Y=15,Z=3),Scale3D=(X=1,Y=1,Z=1))),(Item=(Type=Bone,Name=\"dyn_R_backDress_2\"),Transform=(Rotation=(X=0,Y=0.0872,Z=0,W=0.9962),Translation=(X=-2,Y=15 ...(truncated)"
    WeightMinimum=0
    Weight=@n36.Result
   )
   > &RigUnit_ModifyTransforms(
    ItemToModify="((Item=(Type=Bone,Name=\"dyn_L_backDress_3\"),Transform=(Rotation=(X=0,Y=0.4226,Z=0,W=0.9063),Translation=(X=0,Y=0,Z=0),Scale3D=(X=1,Y=1,Z=1))),(Item=(Type=Bone,Name=\"dyn_R_backDress_3\"),Transform=(Rotation=(X=0,Y=0.4226,Z=0,W=0.9063),Translation=(X=0,Y=0,Z= ...(truncated)"
    WeightMinimum=0
    Weight=@n36.Result
   )
   > &RigUnit_ModifyTransforms(
    ItemToModify="((Item=(Type=Bone,Name=\"dyn_cape_l_02\"),Transform=(Rotation=(X=0,Y=-0.1736,Z=0,W=0.9848),Translation=(X=0,Y=0,Z=0),Scale3D=(X=1,Y=1,Z=1))),(Item=(Type=Bone,Name=\"dyn_cape_r_02\"),Transform=(Rotation=(X=0,Y=-0.1736,Z=0,W=0.9848),Translation=(X=0,Y=0,Z=0),Sca ...(truncated)"
    WeightMinimum=0
    Weight=@n36.Result
   )
   RigVMVariableNode(
    ExecuteContext=()
    Value=&RigUnit_GetTransform(Item=(Type=Bone,Name="root"), Space=LocalSpace, bInitial=False).Transform.Translation
   ) {variable=Root Position}
   n37 = &RigUnit_GetRelativeTransformForItem(Child=(Type=Bone,Name="foot_l"), Parent=(Type=Bone,Name="root"), bChildInitial=False, bParentInitial=False)
   n38 = RigVMVariableNode() {variable=Root Position}
   n39 = &RigVMFunction_AlphaInterp(
    InRange=(Min=7,Max=18)
    InterpSpeedDecreasing=10
    InterpSpeedIncreasing=10
    OutRange=(Min=1,Max=0)
    Scale=1
    bClampResult=true
    bMapRange=true
    Value=&RigVMFunction_MathVectorDistance(
     B=/Script/RigVM.RigVMFunction_MathVectorAdd(A=(X=-35,Y=-20,Z=0), A.Z=@n37.RelativeTransform.Translation.Z, B=@n38.Value) {is=struct RigVMUnitNode, template="Add::Execute(in A,in B,out Result)"}.Result
     A=@n37.RelativeTransform.Translation
    ).Result
   )
   > RigVMVariableNode(Value=@n39.Result) {variable=L Leg Across Side}
   n40 = &RigUnit_GetRelativeTransformForItem(Child=(Type=Bone,Name="foot_r"), Parent=(Type=Bone,Name="root"), bChildInitial=False, bParentInitial=False)
   n41 = RigVMVariableNode() {variable=Root Position}
   n42 = &RigVMFunction_AlphaInterp(
    InRange=(Min=10,Max=28)
    InterpSpeedDecreasing=10
    InterpSpeedIncreasing=10
    OutRange=(Min=1,Max=0)
    Scale=1
    bClampResult=true
    bMapRange=true
    Value=&RigVMFunction_MathVectorDistance(
     B=&RigVMFunction_MathVectorAdd(A=(X=35,Y=-20,Z=-20), A.Z=@n40.RelativeTransform.Translation.Z, B=@n41.Value) {template="Add::Execute(in A,in B,out Result)"}.Result
     A=@n40.RelativeTransform.Translation
    ).Result
   )
   > RigVMVariableNode(Value=@n42.Result) {variable=R Leg Across Side}
   n43 = &RigUnit_GetRelativeTransformForItem(Child=(Type=Bone,Name="foot_l"), Parent=(Type=Bone,Name="root"), bChildInitial=False, bParentInitial=False)
   n44 = RigVMVariableNode() {variable=Root Position}
   n45 = &RigVMFunction_MathVectorAdd(B=(X=-25,Y=-40,Z=7), B.Z=@n43.RelativeTransform.Translation.Z, A=@n44.Value) {template="Add::Execute(in A,in B,out Result)"}
   n46 = &RigVMFunction_AlphaInterp(
    ClampMax=0
    ClampMin=1
    InRange=(Min=10,Max=28)
    InterpSpeedDecreasing=10
    InterpSpeedIncreasing=10
    OutRange=(Min=1,Max=0)
    Scale=1
    bClampResult=false
    bMapRange=false
    Value=&RigVMFunction_MathDoubleRemap(
     SourceMaximum=30
     SourceMinimum=10
     TargetMaximum=0
     TargetMinimum=1
     Value=&RigVMFunction_MathVectorDistance(A=(X=0,Y=0,Z=0), B=@n45.Result, A.X=@n43.RelativeTransform.Translation.X, A.Y=@n43.RelativeTransform.Translation.Y, A.Z=@n44.Value.Z).Result
    ).Result
   )
   > RigVMVariableNode(Value=@n46.Result) {variable=L Leg Across Rear}
   > /Script/RigVM.RigVMFunction_DebugTransformMutableNoSpace(
    Color=(R=1,G=0,B=0,A=1)
    DebugDrawSettings=(DepthPriority=SDPG_Foreground,Lifetime=-1)
    Mode=Axes
    Scale=10
    Thickness=2
    Transform=(Rotation=(X=0,Y=0,Z=0,W=1),Translation=(X=0,Y=0,Z=0),Scale3D=(X=1,Y=1,Z=1))
    WorldOffset=(Rotation=(X=0,Y=0,Z=0,W=1),Translation=(X=0,Y=0,Z=0),Scale3D=(X=1,Y=1,Z=1))
    bEnabled=false
    Transform.Translation=@n45.Result
   ) {is=struct RigVMUnitNode}
   n47 = RigVMVariableNode() {variable=Root Position}
   n48 = &RigVMFunction_MathVectorAdd(B=(X=25,Y=-40,Z=7), A=@n47.Value, B.Z=@n47.Value.Z) {template="Add::Execute(in A,in B,out Result)"}
   n49 = &RigUnit_GetRelativeTransformForItem(Child=(Type=Bone,Name="foot_r"), Parent=(Type=Bone,Name="root"), bChildInitial=False, bParentInitial=False)
   n50 = &RigVMFunction_AlphaInterp(
    ClampMax=0
    ClampMin=1
    InRange=(Min=10,Max=28)
    InterpSpeedDecreasing=10
    InterpSpeedIncreasing=10
    OutRange=(Min=1,Max=0)
    Scale=1
    bClampResult=false
    bMapRange=false
    Value=&RigVMFunction_MathDoubleRemap(
     SourceMaximum=30
     SourceMinimum=10
     TargetMaximum=0
     TargetMinimum=1
     Value=&RigVMFunction_MathVectorDistance(A=(X=0,Y=0,Z=0), B=@n48.Result, A.X=@n49.RelativeTransform.Translation.X, A.Y=@n49.RelativeTransform.Translation.Y, A.Z=@n47.Value.Z).Result
    ).Result
   )
   > RigVMVariableNode(Value=@n50.Result) {variable=R Leg Across Rear}
   n51 = &RigVMFunction_MathTransformMul(
    A=(Rotation=(X=0,Y=0,Z=0,W=1),Translation=(X=-15,Y=23,Z=28),Scale3D=(X=1,Y=1,Z=1))
    B=&RigUnit_GetRelativeTransformForItem(Child=(Type=Bone,Name="pelvis"), Parent=(Type=Bone,Name="root"), bChildInitial=False, bParentInitial=true).RelativeTransform
   ) {template="Multiply::Execute(in A,in B,out Result)"}
   n52 = &RigUnit_GetRelativeTransformForItem(Child=(Type=Bone,Name="foot_l"), Parent=(Type=Bone,Name="root"), bChildInitial=False, bParentInitial=true)
   &RigVMFunction_AlphaInterp(
    InRange=(Min=10,Max=28)
    InterpSpeedDecreasing=10
    InterpSpeedIncreasing=10
    OutRange=(Min=1,Max=0)
    Scale=1
    bClampResult=true
    bMapRange=false
    Value=&RigVMFunction_MathDoubleRemap(
     SourceMaximum=10
     SourceMinimum=1.4
     TargetMaximum=0
     TargetMinimum=1
     Value=&RigVMFunction_MathVectorDistance(A=@n51.Result.Translation, B=@n52.RelativeTransform.Translation).Result
    ).Result
   )
   Inverse = /Script/RigVM.RigVMFunction_MathQuaternionInverse(Value=&RigUnit_GetTransform(Item=(Type=Bone,Name="thigh_l"), Space=LocalSpace, bInitial=False).Transform.Rotation) {is=struct RigVMUnitNode, template="Inverse::Execute(in Value,out Result)"}
   RigVMFunction_MathQuaternionToAxisAndAngle = /Script/RigVM.RigVMFunction_MathQuaternionToAxisAndAngle(Value=/Script/RigVM.RigVMFunction_MathQuaternionMul(B=@Inverse.Result, A=@n5.Value) {is=struct RigVMUnitNode, template="Multiply::Execute(in A,in B,out Result)"}.Result) {is=struct RigVMUnitNode, template="ToAxisAndAngle::Execute(in Value,out Axis,out Angle)"}
   Degrees = /Script/RigVM.RigVMFunction_MathDoubleDeg(Value=@RigVMFunction_MathQuaternionToAxisAndAngle.Angle) {is=struct RigVMUnitNode, template="Degrees::Execute(in Value,out Result)"}
   n53 = &RigVMFunction_MathQuaternionInverse(Value=&RigUnit_GetTransform(Item=(Type=Bone,Name="thigh_r"), Space=LocalSpace, bInitial=False).Transform.Rotation) {template="Inverse::Execute(in Value,out Result)"}
   &RigVMFunction_MathDoubleDeg(Value=&RigVMFunction_MathQuaternionToAxisAndAngle(Value=&RigVMFunction_MathQuaternionMul(B=@n53.Result, A=@n8.Value) {template="Multiply::Execute(in A,in B,out Result)"}.Result) {template="ToAxisAndAngle::Execute(in Value,out Axis,out Angle)"}.Angle) {template="Degrees::Execute(in Value,out Result)"}
   KalmanFilter = &RigVMFunction_KalmanFloat(BufferSize=4, Value=@n6.Result) {template="KalmanFilter::Execute(in Value,in BufferSize,out Result)"}
   &RigVMFunction_KalmanFloat(BufferSize=4, Value=@n9.Result) {template="KalmanFilter::Execute(in Value,in BufferSize,out Result)"}
   &RigUnit_ModifyTransforms(
    ExecutePin=()
    ItemToModify="((Item=(Type=Bone,Name=\"dyn_cape_02\"),Transform=(Rotation=(X=0,Y=-0.0436,Z=0,W=0.999),Translation=(X=0,Y=0,Z=0),Scale3D=(X=1,Y=1,Z=1))),(Item=(Type=Bone,Name=\"dyn_cape_l_02\"),Transform=(Rotation=(X=0,Y=-0.0436,Z=0,W=0.999),Translation=(X=0,Y=0,Z=0),Scale3D ...(truncated)"
    Weight=1
    WeightMinimum=0
   )
   &RigUnit_ModifyTransforms(
    ExecutePin=()
    ItemToModify="((Item=(Type=Bone,Name=\"dyn_L_backDress_1\"),Transform=(Rotation=(X=-0.1006,Y=-0.6259,Z=0.1006,W=0.7668),Translation=(X=-25,Y=0,Z=0),Scale3D=(X=1,Y=1,Z=1))),(Item=(Type=Bone,Name=\"dyn_R_backDress_1\"),Transform=(Rotation=(X=-0.1006,Y=-0.6259,Z=0.1006,W=0.766 ...(truncated)"
    Weight=1
    WeightMinimum=0
   )
   &RigUnit_PrepareForExecution()
   &RigVMFunction_MathDoubleRemap(SourceMaximum=-40, SourceMinimum=-25, TargetMaximum=1, TargetMinimum=0, Value=0)
   &RigVMFunction_MathDoubleRemap(SourceMaximum=40, SourceMinimum=25, TargetMaximum=1, TargetMinimum=0, Value=0)
   RigVMRerouteNode(Value=())
  ]}
]
hierarchy [
 root {loc=<0 0 0>, rot=<0 0 0 1>}
 pelvis {parent=root, loc=<0 -0.5301 98.0571>, rot=<0.0127 0.707 -0.0127 -0.707>}
 spine_01 {parent=pelvis, loc=<6.1208 -0.9328 0>, rot=<0 0 0.034 -0.9994>}
 spine_02 {parent=spine_01, loc=<6.5997 -0.2095 0>, rot=<0 0 0.0238 -0.9997>}
 spine_03 {parent=spine_02, loc=<6.4914 -0.1237 0>, rot=<0 0 0.0149 -0.9999>}
 spine_04 {parent=spine_03, loc=<6.9598 0.0794 0>, rot=<0 0 -0.0132 -0.9999>}
 spine_05 {parent=spine_04, loc=<6.9987 0 0>, rot=<0 0 -0.0015 -1>}
 neck_01 {parent=spine_05, loc=<11.3138 0.1529 0>, rot=<0 0 -0.2003 -0.9797>}
 neck_02 {parent=neck_01, loc=<4.5627 0 0>, rot=<0 0 0.0504 -0.9987>}
 head {parent=neck_02, loc=<5.3914 0 0>, rot=<0 0 0.115 -0.9934>}
 dyn_C_hoodFront {parent=head, loc=<12.5104 3.2958 0.0976>, rot=<0.6954 0.1037 0.709 -0.0548>}
 dyn_L_hoodFront {parent=head, loc=<-6.9797 1.9245 -6.1877>, rot=<-0.1094 -0.8085 -0.1581 -0.5562>}
 dyn_R_hoodFront {parent=head, loc=<-6.9794 1.9245 6.1877>, rot=<0.8085 -0.1094 -0.5562 0.1581>}
 FACIAL_C_FacialRoot {parent=head, loc=<2.1524 -0.746 -0.0002>, rot=<0.4879 -0.5118 -0.4879 -0.5118>}
 FACIAL_L_UnderChin {parent=FACIAL_C_FacialRoot, loc=<3.014 7.5903 2.8085>, rot=<0.4042 -0.354 0.0819 -0.8394>}
 FACIAL_R_UnderChin {parent=FACIAL_C_FacialRoot, loc=<-3.0816 7.5653 2.8044>, rot=<0.4034 0.3555 -0.0838 -0.839>}
 FACIAL_C_Jaw {parent=FACIAL_C_FacialRoot, loc=<-0.0016 2.831 2.3532>, rot=<0.3435 0.0007 -0.001 -0.9392>}
 FACIAL_C_Jawline {parent=FACIAL_C_Jaw, loc=<-0.0198 2.3664 7.0083>, rot=<0.2586 0 0 -0.966>}
 FACIAL_L_Ear {parent=FACIAL_C_FacialRoot, loc=<6.5003 0.7879 0.838>, rot=<-0.0248 -0.0004 -0.0274 -0.9993>}
 FACIAL_R_Ear {parent=FACIAL_C_FacialRoot, loc=<-6.6395 0.5735 0.8631>, rot=<-0.0248 0.001 0.0249 -0.9994>}
 FACIAL_L_Jawline {parent=FACIAL_C_FacialRoot, loc=<4.4219 6.1103 4.5468>, rot=<0.2106 0.1582 -0.056 -0.9631>}
 FACIAL_R_Jawline {parent=FACIAL_C_FacialRoot, loc=<-4.4968 6.0406 4.5416>, rot=<0.211 -0.1571 0.0538 -0.9633>}
 dyn_C_hoodBack {parent=neck_01, loc=<7.1477 -6.2136 0>, rot=<-0.1137 -0.6979 0.1137 -0.6979>}
 clavicle_l {parent=spine_05, loc=<5.4831 -0.1053 -3.2946>, rot=<0.0465 -0.7289 -0.009 -0.6829>}
 upperarm_l {parent=clavicle_l, loc=<9.8658 -1.2136 -0.4105>, rot=<0.0164 -0.3245 -0.0269 -0.9454>}
 lowerarm_l {parent=upperarm_l, loc=<26.465 -0.8345 0.259>, rot=<0.0036 0.0161 -0.2298 -0.9731>}
 hand_l {parent=lowerarm_l, loc=<22.7355 -0.3203 -0.6105>, rot=<-0.6784 -0.0223 0.0735 -0.7307>}
 thumb_01_l {parent=hand_l, loc=<1.3118 -0.3191 -2.459>, rot=<0.6546 -0.4042 0.0664 -0.6354>}
 thumb_02_l {parent=thumb_01_l, loc=<5.9336 -0.0448 0.0401>, rot=<-0.0037 0.0213 0.298 -0.9543>}
 thumb_03_l {parent=thumb_02_l, loc=<2.2223 0.3682 -0.0907>, rot=<0.0118 -0.013 0.0066 -0.9998>}
 index_metacarpal_l {parent=hand_l, loc=<1.6535 0.0183 -2.96>, rot=<0.0158 -0.0124 -0.0173 -0.9996>}
 index_01_l {parent=index_metacarpal_l, loc=<8.3008 0.0357 -0.0524>, rot=<-0.0044 0.0485 0.0924 -0.9945>}
 index_02_l {parent=index_01_l, loc=<3.3587 -0.3004 0.0452>, rot=<-0.002 0.0027 0.3225 -0.9466>}
 index_03_l {parent=index_02_l, loc=<2.5578 -0.123 -0.0252>, rot=<-0.002 0.0067 0.2558 -0.9667>}
 middle_metacarpal_l {parent=hand_l, loc=<2.1856 -0.1725 -0.5984>, rot=<0.006 0.0168 -0.0329 -0.9993>}
 middle_01_l {parent=middle_metacarpal_l, loc=<7.3275 -0.2447 0.0129>, rot=<-0.0208 0.0042 0.0858 -0.9961>}
 middle_02_l {parent=middle_01_l, loc=<4.0123 0.1401 0.1175>, rot=<-0.0218 0.0258 0.3512 -0.9357>}
 middle_03_l {parent=middle_02_l, loc=<2.6077 0.0534 0.1373>, rot=<0.0082 0.0188 0.1708 -0.9851>}
 ring_metacarpal_l {parent=hand_l, loc=<2.2074 -0.3625 1.3638>, rot=<-0.0284 0.0722 -0.0458 -0.9959>}
 ring_01_l {parent=ring_metacarpal_l, loc=<7.0295 -0.6197 -0.1446>, rot=<-0.0489 -0.0252 0.1892 -0.9804>}
 ring_02_l {parent=ring_01_l, loc=<3.6174 0.4647 -0.1027>, rot=<0 0 0.2806 -0.9598>}
 ring_03_l {parent=ring_02_l, loc=<2.4152 0.114 0.0199>, rot=<0 0 0.1189 -0.9929>}
 pinky_metacarpal_l {parent=hand_l, loc=<1.7616 -0.9753 3.3814>, rot=<-0.0397 0.0814 -0.069 -0.9935>}
 pinky_01_l {parent=pinky_metacarpal_l, loc=<7.1709 -0.4831 -0.1094>, rot=<-0.0654 -0.0035 0.2242 -0.9723>}
 pinky_02_l {parent=pinky_01_l, loc=<2.8274 0.0893 -0.1043>, rot=<0.0037 0.0011 0.2854 -0.9584>}
 pinky_03_l {parent=pinky_02_l, loc=<2.2046 0.0307 -0.0078>, rot=<-0.0002 0.0098 0.1445 -0.9895>}
 lowerarm_twist_02_l {parent=lowerarm_l, loc=<10.6707 0 0>, rot=<0 0 0 1>}
 lowerarm_twist_01_l {parent=lowerarm_l, loc=<18.316 0 0>, rot=<0 0 0 1>}
 upperarm_twist_01_l {parent=upperarm_l, loc=<7.2135 -0.1134 0>, rot=<0 0 0 1>}
 upperarm_twist_02_l {parent=upperarm_l, loc=<19.7998 -0.1134 0>, rot=<0 0 0 1>}
 dyn_cape_l_MSTR {parent=spine_05, loc=<1.7124 -9.2171 -16.2328>, rot=<0.1143 -0.6978 0.6934 -0.1386>}
 dyn_cape_l_01 {parent=dyn_cape_l_MSTR, loc=<0.3 0 0>, rot=<0 0 0 1>}
 dyn_cape_l_02 {parent=dyn_cape_l_01, loc=<16.1032 0 -1.5316>, rot=<0 -0.0648 0 -0.9979>}
 dyn_cape_l_03 {parent=dyn_cape_l_02, loc=<16.1032 0 0>, rot=<0 -0.0149 0 -0.9999>}
 dyn_cape_l_04 {parent=dyn_cape_l_03, loc=<16.1032 0 0>, rot=<0 -0.0046 0 -1>}
 dyn_cape_l_05 {parent=dyn_cape_l_04, loc=<16.1032 0 0>, rot=<0 0 0 1>}
 dyn_cape_l_06 {parent=dyn_cape_l_05, loc=<16.1032 0 0>, rot=<0 0 0 1>}
 dyn_cape_l_07 {parent=dyn_cape_l_06, loc=<16.1032 0 0>, rot=<0 0 0 1>}
 clavicle_r {parent=spine_05, loc=<5.4831 -0.1053 3.2946>, rot=<0.7289 0.0465 -0.6829 0.009>}
 upperarm_r {parent=clavicle_r, loc=<-9.8658 1.2136 0.4105>, rot=<0.0164 -0.3245 -0.0269 -0.9454>}
 lowerarm_r {parent=upperarm_r, loc=<-26.465 0.8345 -0.259>, rot=<0.0036 0.0161 -0.2298 -0.9731>}
 hand_r {parent=lowerarm_r, loc=<-22.7355 0.3203 0.6106>, rot=<-0.6784 -0.0223 0.0735 -0.7307>}
 thumb_01_r {parent=hand_r, loc=<-1.3118 0.3191 2.459>, rot=<0.6546 -0.4042 0.0664 -0.6354>}
 thumb_02_r {parent=thumb_01_r, loc=<-5.9336 0.0448 -0.0401>, rot=<-0.0037 0.0213 0.298 -0.9543>}
 thumb_03_r {parent=thumb_02_r, loc=<-2.2223 -0.3682 0.0907>, rot=<0.0118 -0.013 0.0066 -0.9998>}
 index_metacarpal_r {parent=hand_r, loc=<-1.6535 -0.0183 2.96>, rot=<0.0158 -0.0124 -0.0173 -0.9996>}
 index_01_r {parent=index_metacarpal_r, loc=<-8.3008 -0.0357 0.0524>, rot=<-0.0044 0.0485 0.0924 -0.9945>}
 index_02_r {parent=index_01_r, loc=<-3.3587 0.3004 -0.0452>, rot=<-0.002 0.0027 0.3225 -0.9466>}
 index_03_r {parent=index_02_r, loc=<-2.5578 0.123 0.0252>, rot=<-0.002 0.0067 0.2558 -0.9667>}
 middle_metacarpal_r {parent=hand_r, loc=<-2.1856 0.1725 0.5984>, rot=<0.006 0.0168 -0.0329 -0.9993>}
 middle_01_r {parent=middle_metacarpal_r, loc=<-7.3275 0.2447 -0.0129>, rot=<-0.0208 0.0042 0.0858 -0.9961>}
 middle_02_r {parent=middle_01_r, loc=<-4.0123 -0.1401 -0.1175>, rot=<-0.0218 0.0258 0.3512 -0.9357>}
 middle_03_r {parent=middle_02_r, loc=<-2.6077 -0.0534 -0.1373>, rot=<0.0082 0.0188 0.1708 -0.9851>}
 ring_metacarpal_r {parent=hand_r, loc=<-2.2074 0.3625 -1.3638>, rot=<-0.0284 0.0722 -0.0458 -0.9959>}
 ring_01_r {parent=ring_metacarpal_r, loc=<-7.0295 0.6197 0.1446>, rot=<-0.0489 -0.0252 0.1892 -0.9804>}
 ring_02_r {parent=ring_01_r, loc=<-3.6174 -0.4647 0.1027>, rot=<0 0 0.2806 -0.9598>}
 ring_03_r {parent=ring_02_r, loc=<-2.4152 -0.114 -0.0199>, rot=<0 0 0.1189 -0.9929>}
 pinky_metacarpal_r {parent=hand_r, loc=<-1.7616 0.9753 -3.3814>, rot=<-0.0397 0.0814 -0.069 -0.9935>}
 pinky_01_r {parent=pinky_metacarpal_r, loc=<-7.1709 0.4831 0.1094>, rot=<-0.0654 -0.0035 0.2242 -0.9723>}
 pinky_02_r {parent=pinky_01_r, loc=<-2.8274 -0.0893 0.1043>, rot=<0.0037 0.0011 0.2854 -0.9584>}
 pinky_03_r {parent=pinky_02_r, loc=<-2.2046 -0.0307 0.0078>, rot=<-0.0002 0.0098 0.1445 -0.9895>}
 lowerarm_twist_02_r {parent=lowerarm_r, loc=<-10.6707 0 0>, rot=<0 0 0 1>}
 lowerarm_twist_01_r {parent=lowerarm_r, loc=<-18.316 0 0>, rot=<0 0 0 -1>}
 upperarm_twist_01_r {parent=upperarm_r, loc=<-7.2135 0.1134 0>, rot=<0 0 0 -1>}
 upperarm_twist_02_r {parent=upperarm_r, loc=<-19.7998 0.1134 0>, rot=<0 0 0 -1>}
 dyn_cape_r_MSTR {parent=spine_05, loc=<1.7124 -9.2171 16.2328>, rot=<0.1386 -0.6934 0.6978 -0.1143>}
 dyn_cape_r_01 {parent=dyn_cape_r_MSTR, loc=<0.3 0 0>, rot=<0 0 0 1>}
 dyn_cape_r_02 {parent=dyn_cape_r_01, loc=<16.1032 0 -1.5316>, rot=<0 -0.0648 0 -0.9979>}
 dyn_cape_r_03 {parent=dyn_cape_r_02, loc=<16.1032 0 0>, rot=<0 -0.0149 0 -0.9999>}
 dyn_cape_r_04 {parent=dyn_cape_r_03, loc=<16.1033 0 0>, rot=<0 -0.0046 0 -1>}
 dyn_cape_r_05 {parent=dyn_cape_r_04, loc=<16.1033 0 0>, rot=<0 0 0 1>}
 dyn_cape_r_06 {parent=dyn_cape_r_05, loc=<16.1032 0 0>, rot=<0 0 0 1>}
 dyn_cape_r_07 {parent=dyn_cape_r_06, loc=<16.1031 0 0.0001>, rot=<0 0 0 1>}
 attach_backpack {parent=spine_05, loc=<-0.123 -13.473 0>, rot=<-0.0093 -0.707 0.0093 -0.707>}
 Backpack_BR {parent=attach_backpack, type=Null, loc=<0 0.1208 -1.7504>, rot=<-0.0436 0 0 0.999>}
 attach_cape {parent=spine_05, loc=<9.017 -11.714 0>, rot=<0.0616 -0.7044 -0.0616 -0.7044>}
 cape {parent=attach_cape, type=Null, loc=<0 0 0>, rot=<0 0 0 1>}
 dyn_cape_MSTR {parent=spine_05, loc=<1.8395 -10.7979 0>, rot=<0.1265 -0.6957 0.6957 -0.1265>}
 dyn_cape_01 {parent=dyn_cape_MSTR, loc=<0.3 0 0>, rot=<0 0 0 1>}
 dyn_cape_02 {parent=dyn_cape_01, loc=<16.1032 0 -1.5316>, rot=<0 -0.0648 0 -0.9979>}
 dyn_cape_03 {parent=dyn_cape_02, loc=<16.1032 0 0>, rot=<0 -0.0149 0 -0.9999>}
 dyn_cape_04 {parent=dyn_cape_03, loc=<16.1032 0 0>, rot=<0 -0.0046 0 -1>}
 dyn_cape_05 {parent=dyn_cape_04, loc=<16.1032 0 0>, rot=<0 0 0 1>}
 dyn_cape_06 {parent=dyn_cape_05, loc=<16.1032 0 0>, rot=<0 0 0 1>}
 dyn_cape_07 {parent=dyn_cape_06, loc=<16.1032 0 0>, rot=<0 0 0 1>}
 thigh_l {parent=pelvis, loc=<-1.0855 -0.2109 -8.7053>, rot=<0.0009 0.0355 0.0155 -0.9993>}
 calf_l {parent=thigh_l, loc=<-42.5284 -0.0019 -0.7297>, rot=<0.0006 -0.0161 -0.0391 -0.9991>}
 foot_l {parent=calf_l, loc=<-46.7401 -2.5014 -0.592>, rot=<0.0427 -0.0238 0.0353 -0.9982>}
 ball_l {parent=foot_l, loc=<-3.5658 10.8066 -0.1544>, rot=<0.0001 -0.0001 -0.7063 -0.7079>}
 calf_twist_02_l {parent=calf_l, loc=<-19.1546 -1.7943 -0.2638>, rot=<0.0237 0.0095 -0.0035 -0.9997>}
 calf_twist_01_l {parent=calf_l, loc=<-35.8483 -1.9197 -0.5794>, rot=<0.0237 0.0095 -0.0035 -0.9997>}
 dyn_L_frontDress_driver {parent=thigh_l, loc=<1.1065 5.8388 0.0038>, rot=<-0.0208 -0.7434 -0.6683 -0.0169>}
 dyn_L_frontDress_1 {parent=dyn_L_frontDress_driver, loc=<0 0 0>, rot=<0 0 0 1>}
 dyn_L_frontDress_2 {parent=dyn_L_frontDress_1, loc=<21.0771 0 0>, rot=<0 0 0 1>}
 dyn_L_frontDress_3 {parent=dyn_L_frontDress_2, loc=<22.3003 0 0>, rot=<0 0 0 1>}
 dyn_L_frontDress_4 {parent=dyn_L_frontDress_3, loc=<19.4948 0 0>, rot=<0 0 0 1>}
 dyn_L_backDress_driver {parent=thigh_l, loc=<1.6278 -8.7701 -0.697>, rot=<0.0233 -0.7217 0.6917 -0.0128>}
 dyn_L_backDress_1 {parent=dyn_L_backDress_driver, loc=<0 0 0>, rot=<0 0 0 1>}
 dyn_L_backDress_2 {parent=dyn_L_backDress_1, loc=<21.6146 -0.0001 0>, rot=<0 0 0 1>}
 dyn_L_backDress_3 {parent=dyn_L_backDress_2, loc=<23.103 -0.0423 -0.0486>, rot=<0 0 0 1>}
 dyn_L_backDress_4 {parent=dyn_L_backDress_3, loc=<21.3961 -0.1002 -0.1152>, rot=<0 0 0 1>}
 dyn_L_sideDress_driver {parent=thigh_l, loc=<0.9535 -1.7959 -4.9352>, rot=<0.0027 -0.9995 -0.0015 -0.0317>}
 dyn_L_sideDress_1 {parent=dyn_L_sideDress_driver, loc=<0 0 0>, rot=<0 0 0 1>}
 dyn_L_sideDress_2 {parent=dyn_L_sideDress_1, loc=<21.0771 0 0>, rot=<0 0 0 1>}
 dyn_L_sideDress_3 {parent=dyn_L_sideDress_2, loc=<22.3003 0 0>, rot=<0 0 0 1>}
 dyn_L_thighCorrection {parent=thigh_l, loc=<-14.3905 0 0>, rot=<0 0 0 1>}
 thigh_r {parent=pelvis, loc=<-1.0855 -0.2109 8.7053>, rot=<0.0355 -0.0009 0.9993 0.0155>}
 calf_r {parent=thigh_r, loc=<42.5284 0.0019 0.7297>, rot=<0.0006 -0.0161 -0.0391 -0.9991>}
 foot_r {parent=calf_r, loc=<46.7401 2.5014 0.592>, rot=<0.0427 -0.0238 0.0353 -0.9982>}
 ball_r {parent=foot_r, loc=<3.5658 -10.8066 0.1544>, rot=<0.0001 -0.0001 -0.7063 -0.7079>}
 calf_twist_02_r {parent=calf_r, loc=<19.1546 1.7943 0.2638>, rot=<0.0237 0.0095 -0.0035 -0.9997>}
 calf_twist_01_r {parent=calf_r, loc=<35.8483 1.9197 0.5794>, rot=<0.0237 0.0095 -0.0035 -0.9997>}
 dyn_R_frontDress_driver {parent=thigh_r, loc=<-1.1065 -5.8388 -0.0038>, rot=<-0.0208 -0.7434 -0.6683 -0.0169>}
 dyn_R_frontDress_1 {parent=dyn_R_frontDress_driver, loc=<0 0 0>, rot=<0 0 0 1>}
 dyn_R_frontDress_2 {parent=dyn_R_frontDress_1, loc=<-21.077 0 0>, rot=<0 0 0 1>}
 dyn_R_frontDress_3 {parent=dyn_R_frontDress_2, loc=<-22.3003 0 0>, rot=<0 0 0 1>}
 dyn_R_frontDress_4 {parent=dyn_R_frontDress_3, loc=<-19.4948 0 0>, rot=<0 0 0 1>}
 dyn_R_backDress_driver {parent=thigh_r, loc=<-1.6277 8.7701 0.697>, rot=<0.0233 -0.7217 0.6917 -0.0128>}
 dyn_R_backDress_1 {parent=dyn_R_backDress_driver, loc=<0 0 0>, rot=<0 0 0 1>}
 dyn_R_backDress_2 {parent=dyn_R_backDress_1, loc=<-21.6145 0 0>, rot=<0 0 0 1>}
 dyn_R_backDress_3 {parent=dyn_R_backDress_2, loc=<-23.103 0.0423 0.0486>, rot=<0 0 0 1>}
 dyn_R_backDress_4 {parent=dyn_R_backDress_3, loc=<-21.3961 0.1003 0.1151>, rot=<0 0 0 -1>}
 dyn_R_sideDress_driver {parent=thigh_r, loc=<-0.9535 1.7959 4.9352>, rot=<0.0027 -0.9995 -0.0015 -0.0317>}
 dyn_R_sideDress_1 {parent=dyn_R_sideDress_driver, loc=<0 0 0>, rot=<0 0 0 -1>}
 dyn_R_sideDress_2 {parent=dyn_R_sideDress_1, loc=<-21.0771 0 -0.0001>, rot=<0 0 0 1>}
 dyn_R_sideDress_3 {parent=dyn_R_sideDress_2, loc=<-22.3004 0 0.0001>, rot=<0 0 0 1>}
 dyn_R_thighCorrection {parent=thigh_r, loc=<14.3904 0 0>, rot=<0 0 0 -1>}
 dyn_C_frontDress_driver {parent=pelvis, loc=<0.3451 7.3201 0>, rot=<0.0293 0.7065 0.7065 0.0293>}
 dyn_C_frontDress_1 {parent=dyn_C_frontDress_driver, loc=<0 0 0>, rot=<0 0 0 1>}
 dyn_C_frontDress_2 {parent=dyn_C_frontDress_1, loc=<21.0771 0 0>, rot=<0 0 0 1>}
 dyn_C_frontDress_3 {parent=dyn_C_frontDress_2, loc=<22.3003 0 0>, rot=<0 0 0 1>}
 dyn_C_frontDress_4 {parent=dyn_C_frontDress_3, loc=<19.4948 0 0>, rot=<0 0 0 1>}
 dyn_C_backDress_driver {parent=pelvis, loc=<-0.0917 -10.7096 0>, rot=<-0.707 -0.0102 -0.0102 -0.707>}
 dyn_C_backDress_1 {parent=dyn_C_backDress_driver, loc=<0 0 0>, rot=<0 0 0 1>}
 dyn_C_backDress_2 {parent=dyn_C_backDress_1, loc=<-21.153 0 -0.6363>, rot=<0 0 0 1>}
 dyn_C_backDress_3 {parent=dyn_C_backDress_2, loc=<-23.0674 0 0.1648>, rot=<0 0 0 1>}
 dyn_C_backDress_4 {parent=dyn_C_backDress_3, loc=<-21.3674 0 0.0849>, rot=<0 0 0 1>}
]
variables [
 L Leg Across Rear {type=float, default=0}
 L Leg Across Side {type=float, default=0}
 L thigh foward alpha {type=double, default=0}
 L thigh inward alpha {type=double, default=0}
 R Leg Across Rear {type=float, default=0}
 R Leg Across Side {type=float, default=0}
 R thigh forwardAlpha {type=double, default=0}
 R thigh inward alpha {type=double, default=0}
 Root Position {type=FVector, default=(X=0,Y=0,Z=0)}
 isAccelerating2D {type=bool, default=False}
 isCrouchMoving {type=bool, default=False}
 isCrouchOverall {type=bool, default=False}
 isCustomIdlePlaying {type=double, default=0}
 isFrontend {type=bool, default=False}
 isJumping {type=bool, default=False}
 isKneeSliding {type=bool, default=False}
 isPlayingEmote {type=bool, default=False}
 isRBANenabled {type=bool, default=False}
 isSkydiveFloating {type=bool, default=False}
 isSurferGliderDeployed {type=bool, default=False}
 isWearingBackbling {type=bool, default=False}
 previous L thigh quat {type=FQuat, default=(X=0,Y=0,Z=0,W=1)}
 previous R thigh quat {type=FQuat, default=(X=0,Y=0,Z=0,W=1)}
]
