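# Post-process style AnimBP for the F_MED_ForwardLake_DireLab outfit, adding secondary motion to hood, dress and cape.
# It copies the parent mesh pose through a Control Rig, offsets cape bones, constrains hood and backpack bones, then caches after hood AnimDynamics.
# Dress and cape AnimDynamics chains build on that cache.
# RBAN/cloth flags choose between AnimDynamics, full RBAN, dress-only RBAN or the start pose, feeding a state-driven post-sim Control Rig.
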AnimBlueprint /BRCosmetics/Characters/Player/Female/Medium/Bodies/F_MED_ForwardLake_DireLab/Meshes/F_MED_ForwardLake_DireLab_AnimBP
target_skeleton=&/F_MED_ForwardLake_DireLab_Skeleton
graphs [
 AnimGraph {nodes [
   n1 = Use cached pose 'Full_RBAN_cache'()
   n2 = Get RBAN On Cloth Off() -> (RBAN On Cloth Off:bool)
   n3 = Blend Poses by bool(
    BlendPose_0=Use cached pose 'onlyDress RBAN_cache'()
    BlendPose_1=Use cached pose 'Start_cache'()
    bActiveValue=Get RBAN On Cloth On() -> (RBAN On Cloth On:bool)
   ) {blendType=Linear}
   n4 = Control Rig!(
    Source=Blend Poses by bool(
     BlendPose_0=Use cached pose 'animDyn Cache'()
     BlendPose_1=Blend Poses by bool(BlendPose_0=@n1, BlendPose_1=@n3, bActiveValue=@n2) {blendType=Linear}
     bActiveValue=Get RBAN Off Cloth Off() -> (RBAN Off Cloth Off:bool)
    ) {blendType=Linear}
    Alpha
    isCrouchMoving:bool=Get crouchMoving() -> (crouchMoving:bool)
    isCrouchOverall:bool=Get bIsCrouching()
    isSkydiveFloating:bool=Get bIsSkydiveFloating()
    isSurferGliderDeployed:bool=Get isGliderSurfing() -> (isGliderSurfing:bool)
    isJumping:bool=Get getCustomJump() -> (getCustomJump:bool)
    isAccelerating2D:bool=Get bIsAccelerating2D()
    isCustomIdlePlaying:double=Get isPlayingFrontendIdle() -> (isPlayingFrontendIdle:float)
    isWearingBackbling:bool=Get Is Wearing Backbling()
    isPlayingEmote:bool=Get bIsPlayingEmote()
    isFrontend:bool=Get bIsInFrontEnd()
    isRBANenabled:bool=Get bRigidBodyEnabled()
    isKneeSliding:bool=Get bIsKneeSliding()
   ) -> (Pose) {
    alphaBoolBlend {blendOption=Linear}
    controlRigAssetReference {blueprintRigClass=&/F_MED_ForwardLake_DireLab_PostSim_CtrlRig.F_MED_ForwardLake_DireLab_PostSim_CtrlRig_C}
   }
   Output Pose(Result=@n4) {layerGroup=None}
   n5 = Control Rig!(
    Source=Copy Pose From Mesh() {bUseAttachedParent=true}
    Alpha
    isCrouching:bool=Get bIsCrouching()
    rigidBodyEnabled:bool=Get bRigidBodyEnabled()
    isSkydiveFloating=Get bIsSkydiveFloating()
    isAccelerating2D=Get bIsAccelerating2D()
    isCustomIdlePlaying=Get isPlayingFrontendIdle()
   ) -> (Pose) {
    alphaBoolBlend {blendOption=Linear}
    controlRigAssetReference {blueprintRigClass=&/F_MED_ForwardLake_DireLab_CtrlRig.F_MED_ForwardLake_DireLab_CtrlRig_C}
   }
   n6 = "Transform (Modify) Bone - Bone: dyn_cape_l_03"!(ComponentPose=Local To Component(LocalPose=@n5), Alpha=1) -> (Pose) {
    rotationMode=BMM_Additive
    rotationSpace=BCS_BoneSpace
    alphaBoolBlend {blendOption=Linear}
    alphaScaleBiasClamp {bInterpResult=true, interpSpeedDecreasing=4, interpSpeedIncreasing=4}
    boneToModify {boneName=dyn_cape_l_03}
    rotation {pitch=30}
   }
   n7 = "Transform (Modify) Bone - Bone: dyn_cape_03"!(ComponentPose=@n6, Alpha=1) -> (Pose) {
    rotationMode=BMM_Additive
    rotationSpace=BCS_BoneSpace
    alphaBoolBlend {blendOption=Linear}
    alphaScaleBiasClamp {bInterpResult=true, interpSpeedDecreasing=4, interpSpeedIncreasing=4}
    boneToModify {boneName=dyn_cape_03}
    rotation {pitch=30}
   }
   n8 = "Transform (Modify) Bone - Bone: dyn_cape_r_03"!(ComponentPose=@n7, Alpha=1) -> (Pose) {
    rotationMode=BMM_Additive
    rotationSpace=BCS_BoneSpace
    alphaBoolBlend {blendOption=Linear}
    alphaScaleBiasClamp {bInterpResult=true, interpSpeedDecreasing=4, interpSpeedIncreasing=4}
    boneToModify {boneName=dyn_cape_r_03}
    rotation {pitch=30}
   }
   n9 = "Transform (Modify) Bone - Bone: dyn_cape_l_04"!(ComponentPose=@n8, Alpha=1) -> (Pose) {
    rotationMode=BMM_Additive
    rotationSpace=BCS_BoneSpace
    alphaBoolBlend {blendOption=Linear}
    alphaScaleBiasClamp {bInterpResult=true, interpSpeedDecreasing=4, interpSpeedIncreasing=4}
    boneToModify {boneName=dyn_cape_l_04}
    rotation {pitch=30}
   }
   n10 = "Transform (Modify) Bone - Bone: dyn_cape_04"!(ComponentPose=@n9, Alpha=1) -> (Pose) {
    rotationMode=BMM_Additive
    rotationSpace=BCS_BoneSpace
    alphaBoolBlend {blendOption=Linear}
    alphaScaleBiasClamp {bInterpResult=true, interpSpeedDecreasing=4, interpSpeedIncreasing=4}
    boneToModify {boneName=dyn_cape_04}
    rotation {pitch=30}
   }
   n11 = "Transform (Modify) Bone - Bone: dyn_cape_r_04"!(ComponentPose=@n10, Alpha=1) -> (Pose) {
    rotationMode=BMM_Additive
    rotationSpace=BCS_BoneSpace
    alphaBoolBlend {blendOption=Linear}
    alphaScaleBiasClamp {bInterpResult=true, interpSpeedDecreasing=4, interpSpeedIncreasing=4}
    boneToModify {boneName=dyn_cape_r_04}
    rotation {pitch=30}
   }
   n12 = "Transform (Modify) Bone - Bone: dyn_cape_l_05"!(ComponentPose=@n11, Alpha=1) -> (Pose) {
    rotationMode=BMM_Additive
    rotationSpace=BCS_BoneSpace
    alphaBoolBlend {blendOption=Linear}
    alphaScaleBiasClamp {bInterpResult=true, interpSpeedDecreasing=4, interpSpeedIncreasing=4}
    boneToModify {boneName=dyn_cape_l_05}
    rotation {pitch=30}
   }
   n13 = "Transform (Modify) Bone - Bone: dyn_cape_05"!(ComponentPose=@n12, Alpha=1) -> (Pose) {
    rotationMode=BMM_Additive
    rotationSpace=BCS_BoneSpace
    alphaBoolBlend {blendOption=Linear}
    alphaScaleBiasClamp {bInterpResult=true, interpSpeedDecreasing=4, interpSpeedIncreasing=4}
    boneToModify {boneName=dyn_cape_05}
    rotation {pitch=30}
   }
   n14 = "Transform (Modify) Bone - Bone: dyn_cape_r_05"!(ComponentPose=@n13, Alpha=1) -> (Pose) {
    rotationMode=BMM_Additive
    rotationSpace=BCS_BoneSpace
    alphaBoolBlend {blendOption=Linear}
    alphaScaleBiasClamp {bInterpResult=true, interpSpeedDecreasing=4, interpSpeedIncreasing=4}
    boneToModify {boneName=dyn_cape_r_05}
    rotation {pitch=30}
   }
   n15 = "Transform (Modify) Bone - Bone: dyn_cape_l_06"!(ComponentPose=@n14, Alpha=1) -> (Pose) {
    rotationMode=BMM_Additive
    rotationSpace=BCS_BoneSpace
    alphaBoolBlend {blendOption=Linear}
    alphaScaleBiasClamp {bInterpResult=true, interpSpeedDecreasing=4, interpSpeedIncreasing=4}
    boneToModify {boneName=dyn_cape_l_06}
    rotation {pitch=30}
   }
   n16 = "Transform (Modify) Bone - Bone: dyn_cape_06"!(ComponentPose=@n15, Alpha=1) -> (Pose) {
    rotationMode=BMM_Additive
    rotationSpace=BCS_BoneSpace
    alphaBoolBlend {blendOption=Linear}
    alphaScaleBiasClamp {bInterpResult=true, interpSpeedDecreasing=4, interpSpeedIncreasing=4}
    boneToModify {boneName=dyn_cape_06}
    rotation {pitch=30}
   }
   n17 = "Transform (Modify) Bone - Bone: dyn_cape_r_06"!(ComponentPose=@n16, Alpha=1) -> (Pose) {
    rotationMode=BMM_Additive
    rotationSpace=BCS_BoneSpace
    alphaBoolBlend {blendOption=Linear}
    alphaScaleBiasClamp {bInterpResult=true, interpSpeedDecreasing=4, interpSpeedIncreasing=4}
    boneToModify {boneName=dyn_cape_r_06}
    rotation {pitch=30}
   }
   n18 = "Constraint - dyn_L_hoodFront (PP)"!(ComponentPose=@n17, Alpha=1, ConstraintWeights_0=0.5, ConstraintWeights_1:float=0.5) -> (Pose) {
    alphaBoolBlend {blendOption=Linear}
    boneToModify {boneName=dyn_L_hoodFront}
    constraintSetup [{targetBone {boneName=head}}, {targetBone {boneName=neck_02}}]
    constraintWeights [1,1]
   }
   n19 = "Constraint - dyn_R_hoodFront (PP)"!(ComponentPose=@n18, Alpha=1, ConstraintWeights_0=0.5, ConstraintWeights_1:float=0.5) -> (Pose) {
    alphaBoolBlend {blendOption=Linear}
    boneToModify {boneName=dyn_R_hoodFront}
    constraintSetup [{targetBone {boneName=head}}, {targetBone {boneName=neck_02}}]
    constraintWeights [1,1]
   }
   n20 = "Constraint - attach_backpack (PPPPPPPP)"!(
    ComponentPose=@n19
    Alpha=1
    ConstraintWeights_0=0.0511
    ConstraintWeights_1:float=0.1359
    ConstraintWeights_2:float=0.0307
    ConstraintWeights_3:float=0.2455
    ConstraintWeights_4:float=0.2632
    ConstraintWeights_5:float=0.0502
    ConstraintWeights_6:float=0.1356
    ConstraintWeights_7:float=0.0878
   ) -> (Pose) {
    alphaBoolBlend {blendOption=Linear}
    boneToModify {boneName=attach_backpack}
    constraintSetup [
     {targetBone {boneName=dyn_cape_r_02}}
     {targetBone {boneName=dyn_cape_l_01}}
     {targetBone {boneName=spine_05}}
     {targetBone {boneName=dyn_cape_01}}
     {targetBone {boneName=dyn_cape_02}}
     {targetBone {boneName=dyn_cape_l_02}}
     {targetBone {boneName=dyn_cape_r_01}}
     {targetBone {boneName=spine_04}}
    ]
    constraintWeights [1,1,1,1,1,1,1,1]
   }
   n21 = "Transform (Modify) Bone - Bone: attach_backpack"!(ComponentPose=@n20, Alpha=1, Rotation="0,0,30") -> (Pose) {
    rotationMode=BMM_Additive
    alphaBoolBlend {blendOption=Linear}
    alphaScaleBiasClamp {bInterpResult=true}
    boneToModify {boneName=attach_backpack}
    rotation {roll=30}
   }
   n22 = Anim Dynamics - Bone: dyn_L_hoodFront(ComponentPose=@n21, Alpha=1, ExternalForce="0,0,0") {
    angularDampingOverride=1
    angularSpringConstant=1000
    bAngularSpring=true
    bOverrideAngularDamping=true
    bUseGravityOverride=true
    gravityScale=0.35
    simulationSpace=BoneRelative
    alphaBoolBlend {blendOption=Linear}
    boundBone {boneName=dyn_L_hoodFront}
    componentAppliedLinearAccClamp=<3000 3000 3000>
    componentLinearAccScale=<1 1 1>
    componentLinearVelScale=<1 1 1>
    gravityOverride {z=-980}
    physicsBodyDefinitions [{
      boundBone {boneName=dyn_L_hoodFront}
      boxExtents=<13 13 13>
      constraintSetup {angularLimitsMax=<20 10 30>, angularLimitsMin=<-20 -20 -30>, angularTarget=<1 0 0>}
      localJointOffset=<2 0 0>
     }]
    relativeSpaceBone {boneName=root}
   }
   n23 = Anim Dynamics - Bone: dyn_R_hoodFront(ComponentPose=@n22, Alpha=1, ExternalForce="0,0,0") {
    angularDampingOverride=1
    angularSpringConstant=1000
    bAngularSpring=true
    bOverrideAngularDamping=true
    bUseGravityOverride=true
    gravityScale=0.35
    simulationSpace=BoneRelative
    alphaBoolBlend {blendOption=Linear}
    boundBone {boneName=dyn_R_hoodFront}
    componentAppliedLinearAccClamp=<3000 3000 3000>
    componentLinearAccScale=<1 1 1>
    componentLinearVelScale=<1 1 1>
    gravityOverride {z=-980}
    physicsBodyDefinitions [{
      boundBone {boneName=dyn_R_hoodFront}
      boxExtents=<13 13 13>
      constraintSetup {angularLimitsMax=<20 10 30>, angularLimitsMin=<-20 -20 -30>, angularTarget=<1 0 0>}
      localJointOffset=<-2 0 0>
     }]
    relativeSpaceBone {boneName=root}
   }
   n24 = Anim Dynamics - Bone: dyn_C_hoodBack(ComponentPose=@n23, Alpha=1, ExternalForce="0,0,0") {
    angularDampingOverride=1
    angularSpringConstant=1000
    bAngularSpring=true
    bOverrideAngularDamping=true
    bUseGravityOverride=true
    gravityScale=0.35
    simulationSpace=BoneRelative
    alphaBoolBlend {blendOption=Linear}
    boundBone {boneName=dyn_C_hoodBack}
    componentAppliedLinearAccClamp=<3000 3000 3000>
    componentLinearAccScale=<1 1 1>
    componentLinearVelScale=<1 1 1>
    gravityOverride {z=-980}
    physicsBodyDefinitions [{
      boundBone {boneName=dyn_C_hoodBack}
      boxExtents=<13 13 13>
      constraintSetup {angularLimitsMax=<15 5 15>, angularLimitsMin=<-15 -5 -15>, angularTarget=<1 0 0>}
      localJointOffset=<0 -3 0>
     }]
    relativeSpaceBone {boneName=neck_01}
   }
   n25 = Anim Dynamics - Bone: dyn_C_hoodFront(ComponentPose=@n24, Alpha=1, ExternalForce="0,0,0") {
    angularDampingOverride=1
    angularSpringConstant=7000
    bAngularSpring=true
    bOverrideAngularDamping=true
    bUseGravityOverride=true
    gravityScale=0.35
    simulationSpace=BoneRelative
    alphaBoolBlend {blendOption=Linear}
    boundBone {boneName=dyn_C_hoodFront}
    componentAppliedLinearAccClamp=<3000 3000 3000>
    componentLinearAccScale=<1 1 1>
    componentLinearVelScale=<1 1 1>
    gravityOverride {z=-980}
    physicsBodyDefinitions [{
      boundBone {boneName=dyn_C_hoodFront}
      boxExtents=<13 13 13>
      constraintSetup {angularLimitsMax=<30 5 5>, angularLimitsMin=<0 -5 -5>, angularTarget=<1 0 0>}
      localJointOffset=<0 -3 0>
     }]
    relativeSpaceBone {boneName=neck_01}
   }
   Save cached pose 'Start_cache'(Pose=Component To Local(ComponentPose=@n25)) {cacheName=Start_cache}
   n26 = Anim Dynamics - Bone: dyn_L_frontDress_1!(
    SimulationSpace:AnimPhysSimSpaceType Enum
    ComponentPose=Local To Component(LocalPose=Use cached pose 'Start_cache'())
    Alpha=1
    GravityOverride:Vector="0,0,-980"
   ) -> (Pose) {
    angularDampingOverride=1
    angularSpringConstant=5000
    bAngularSpring=true
    bOverrideAngularDamping=true
    bUseGravityOverride=true
    bUsePlanarLimit=false
    gravityScale=0.35
    simulationSpace=RootRelative
    alphaBoolBlend {blendOption=Linear}
    boundBone {boneName=dyn_L_frontDress_1}
    componentAppliedLinearAccClamp=<3000 3000 3000>
    componentLinearAccScale=<2 2 2>
    componentLinearVelScale=<2 2 2>
    gravityOverride {z=-980}
    physicsBodyDefinitions [{
      boundBone {boneName=dyn_L_frontDress_1}
      boxExtents=<35 35 35>
      constraintSetup {angularLimitsMax=<0 0 0>, angularLimitsMin=<0 -10 0>, angularTarget=<1 0 0>}
      localJointOffset=<5 0 0>
     }]
    relativeSpaceBone {boneName=pelvis}
   }
   n27 = Anim Dynamics - Bone: dyn_L_frontDress_2!(SimulationSpace:AnimPhysSimSpaceType Enum, ComponentPose=@n26, Alpha=1, GravityOverride:Vector="0,0,-980") -> (Pose) {
    angularDampingOverride=1
    angularSpringConstant=5000
    bAngularSpring=true
    bOverrideAngularDamping=true
    bUseGravityOverride=true
    bUsePlanarLimit=false
    gravityScale=0.35
    simulationSpace=RootRelative
    alphaBoolBlend {blendOption=Linear}
    boundBone {boneName=dyn_L_frontDress_2}
    componentAppliedLinearAccClamp=<3000 3000 3000>
    componentLinearAccScale=<2 2 2>
    componentLinearVelScale=<2 2 2>
    gravityOverride {z=-980}
    physicsBodyDefinitions [{
      boundBone {boneName=dyn_L_frontDress_2}
      boxExtents=<35 35 35>
      constraintSetup {angularLimitsMax=<0 10 0>, angularLimitsMin=<0 -10 0>, angularTarget=<1 0 0>}
      localJointOffset=<5 0 0>
     }]
    relativeSpaceBone {boneName=pelvis}
   }
   n28 = Anim Dynamics - Bone: dyn_L_frontDress_3!(SimulationSpace:AnimPhysSimSpaceType Enum, ComponentPose=@n27, Alpha=1, GravityOverride:Vector="0,0,-980") -> (Pose) {
    angularDampingOverride=1
    angularSpringConstant=5000
    bAngularSpring=true
    bOverrideAngularDamping=true
    bUseGravityOverride=true
    bUsePlanarLimit=false
    gravityScale=0.35
    simulationSpace=RootRelative
    alphaBoolBlend {blendOption=Linear}
    boundBone {boneName=dyn_L_frontDress_3}
    componentAppliedLinearAccClamp=<3000 3000 3000>
    componentLinearAccScale=<2 2 2>
    componentLinearVelScale=<2 2 2>
    gravityOverride {z=-980}
    physicsBodyDefinitions [{
      boundBone {boneName=dyn_L_frontDress_3}
      boxExtents=<35 35 35>
      constraintSetup {angularLimitsMax=<0 20 0>, angularLimitsMin=<0 -10 0>, angularTarget=<1 0 0>}
      localJointOffset=<5 0 0>
     }]
    relativeSpaceBone {boneName=pelvis}
   }
   n29 = Anim Dynamics - Bone: dyn_L_frontDress_4!(SimulationSpace:AnimPhysSimSpaceType Enum, ComponentPose=@n28, Alpha=1, GravityOverride:Vector="0,0,-980") -> (Pose) {
    angularDampingOverride=1
    angularSpringConstant=5000
    bAngularSpring=true
    bOverrideAngularDamping=true
    bUseGravityOverride=true
    bUsePlanarLimit=false
    gravityScale=0.35
    simulationSpace=RootRelative
    alphaBoolBlend {blendOption=Linear}
    boundBone {boneName=dyn_L_frontDress_4}
    componentAppliedLinearAccClamp=<3000 3000 3000>
    componentLinearAccScale=<2 2 2>
    componentLinearVelScale=<2 2 2>
    gravityOverride {z=-980}
    physicsBodyDefinitions [{
      boundBone {boneName=dyn_L_frontDress_4}
      boxExtents=<35 35 35>
      constraintSetup {angularLimitsMax=<0 30 0>, angularLimitsMin=<0 -10 0>, angularTarget=<1 0 0>}
      localJointOffset=<5 0 0>
     }]
    relativeSpaceBone {boneName=pelvis}
   }
   n30 = Anim Dynamics - Bone: dyn_L_sideDress_1!(SimulationSpace:AnimPhysSimSpaceType Enum, ComponentPose=@n29, Alpha=1, GravityOverride:Vector="0,0,-980") -> (Pose) {
    angularDampingOverride=1
    angularSpringConstant=5000
    bAngularSpring=true
    bOverrideAngularDamping=true
    bUseGravityOverride=true
    bUsePlanarLimit=false
    gravityScale=0.35
    simulationSpace=RootRelative
    alphaBoolBlend {blendOption=Linear}
    boundBone {boneName=dyn_L_sideDress_1}
    componentAppliedLinearAccClamp=<3000 3000 3000>
    componentLinearAccScale=<2 2 2>
    componentLinearVelScale=<2 2 2>
    gravityOverride {z=-980}
    physicsBodyDefinitions [{
      boundBone {boneName=dyn_L_sideDress_1}
      boxExtents=<35 35 35>
      constraintSetup {angularLimitsMax=<0 0 0>, angularLimitsMin=<0 -15 0>, angularTarget=<1 0 0>}
      localJointOffset=<5 0 0>
     }]
    relativeSpaceBone {boneName=pelvis}
   }
   n31 = Anim Dynamics - Bone: dyn_L_sideDress_2!(SimulationSpace:AnimPhysSimSpaceType Enum, ComponentPose=@n30, Alpha=1, GravityOverride:Vector="0,0,-980") -> (Pose) {
    angularDampingOverride=1
    angularSpringConstant=5000
    bAngularSpring=true
    bOverrideAngularDamping=true
    bUseGravityOverride=true
    bUsePlanarLimit=false
    gravityScale=0.35
    simulationSpace=RootRelative
    alphaBoolBlend {blendOption=Linear}
    boundBone {boneName=dyn_L_sideDress_2}
    componentAppliedLinearAccClamp=<3000 3000 3000>
    componentLinearAccScale=<2 2 2>
    componentLinearVelScale=<2 2 2>
    gravityOverride {z=-980}
    physicsBodyDefinitions [{
      boundBone {boneName=dyn_L_sideDress_2}
      boxExtents=<35 35 35>
      constraintSetup {angularLimitsMax=<0 0 0>, angularLimitsMin=<0 -15 0>, angularTarget=<1 0 0>}
      localJointOffset=<5 0 0>
     }]
    relativeSpaceBone {boneName=pelvis}
   }
   n32 = Anim Dynamics - Bone: dyn_L_sideDress_3!(SimulationSpace:AnimPhysSimSpaceType Enum, ComponentPose=@n31, Alpha=1, GravityOverride:Vector="0,0,-980") -> (Pose) {
    angularDampingOverride=1
    angularSpringConstant=5000
    bAngularSpring=true
    bOverrideAngularDamping=true
    bUseGravityOverride=true
    bUsePlanarLimit=false
    gravityScale=0.35
    simulationSpace=RootRelative
    alphaBoolBlend {blendOption=Linear}
    boundBone {boneName=dyn_L_sideDress_3}
    componentAppliedLinearAccClamp=<3000 3000 3000>
    componentLinearAccScale=<2 2 2>
    componentLinearVelScale=<2 2 2>
    gravityOverride {z=-980}
    physicsBodyDefinitions [{
      boundBone {boneName=dyn_L_sideDress_3}
      boxExtents=<35 35 35>
      constraintSetup {angularLimitsMax=<0 0 0>, angularLimitsMin=<0 -15 0>, angularTarget=<1 0 0>}
      localJointOffset=<5 0 0>
     }]
    relativeSpaceBone {boneName=pelvis}
   }
   n33 = Anim Dynamics - Bone: dyn_L_backDress_1!(SimulationSpace:AnimPhysSimSpaceType Enum, ComponentPose=@n32, Alpha=1, GravityOverride:Vector="0,0,-980") -> (Pose) {
    angularDampingOverride=1
    angularSpringConstant=5000
    bAngularSpring=true
    bOverrideAngularDamping=true
    bUseGravityOverride=true
    bUsePlanarLimit=false
    gravityScale=0.35
    simulationSpace=RootRelative
    alphaBoolBlend {blendOption=Linear}
    boundBone {boneName=dyn_L_backDress_1}
    componentAppliedLinearAccClamp=<3000 3000 3000>
    componentLinearAccScale=<2 2 2>
    componentLinearVelScale=<2 2 2>
    gravityOverride {z=-980}
    physicsBodyDefinitions [{
      boundBone {boneName=dyn_L_backDress_1}
      boxExtents=<35 35 35>
      constraintSetup {angularLimitsMax=<0 0 0>, angularLimitsMin=<0 -15 0>, angularTarget=<1 0 0>}
      localJointOffset=<5 0 0>
     }]
    relativeSpaceBone {boneName=pelvis}
   }
   n34 = Anim Dynamics - Bone: dyn_L_backDress_2!(SimulationSpace:AnimPhysSimSpaceType Enum, ComponentPose=@n33, Alpha=1, GravityOverride:Vector="0,0,-980") -> (Pose) {
    angularDampingOverride=1
    angularSpringConstant=5000
    bAngularSpring=true
    bOverrideAngularDamping=true
    bUseGravityOverride=true
    bUsePlanarLimit=false
    gravityScale=0.35
    simulationSpace=RootRelative
    alphaBoolBlend {blendOption=Linear}
    boundBone {boneName=dyn_L_backDress_2}
    componentAppliedLinearAccClamp=<3000 3000 3000>
    componentLinearAccScale=<2 2 2>
    componentLinearVelScale=<2 2 2>
    gravityOverride {z=-980}
    physicsBodyDefinitions [{
      boundBone {boneName=dyn_L_backDress_2}
      boxExtents=<35 35 35>
      constraintSetup {angularLimitsMax=<0 0 0>, angularLimitsMin=<0 -15 0>, angularTarget=<1 0 0>}
      localJointOffset=<5 0 0>
     }]
    relativeSpaceBone {boneName=pelvis}
   }
   n35 = Anim Dynamics - Bone: dyn_L_backDress_3!(SimulationSpace:AnimPhysSimSpaceType Enum, ComponentPose=@n34, Alpha=1, GravityOverride:Vector="0,0,-980") -> (Pose) {
    angularDampingOverride=1
    angularSpringConstant=5000
    bAngularSpring=true
    bOverrideAngularDamping=true
    bUseGravityOverride=true
    bUsePlanarLimit=false
    gravityScale=0.35
    simulationSpace=RootRelative
    alphaBoolBlend {blendOption=Linear}
    boundBone {boneName=dyn_L_backDress_3}
    componentAppliedLinearAccClamp=<3000 3000 3000>
    componentLinearAccScale=<2 2 2>
    componentLinearVelScale=<2 2 2>
    gravityOverride {z=-980}
    physicsBodyDefinitions [{
      boundBone {boneName=dyn_L_backDress_3}
      boxExtents=<35 35 35>
      constraintSetup {angularLimitsMax=<0 0 0>, angularLimitsMin=<0 -15 0>, angularTarget=<1 0 0>}
      localJointOffset=<5 0 0>
     }]
    relativeSpaceBone {boneName=pelvis}
   }
   n36 = Anim Dynamics - Bone: dyn_L_backDress_4!(SimulationSpace:AnimPhysSimSpaceType Enum, ComponentPose=@n35, Alpha=1, GravityOverride:Vector="0,0,-980") -> (Pose) {
    angularDampingOverride=1
    angularSpringConstant=5000
    bAngularSpring=true
    bOverrideAngularDamping=true
    bUseGravityOverride=true
    bUsePlanarLimit=false
    gravityScale=0.35
    simulationSpace=RootRelative
    alphaBoolBlend {blendOption=Linear}
    boundBone {boneName=dyn_L_backDress_4}
    componentAppliedLinearAccClamp=<3000 3000 3000>
    componentLinearAccScale=<2 2 2>
    componentLinearVelScale=<2 2 2>
    gravityOverride {z=-980}
    physicsBodyDefinitions [{
      boundBone {boneName=dyn_L_backDress_4}
      boxExtents=<35 35 35>
      constraintSetup {angularLimitsMax=<0 0 0>, angularLimitsMin=<0 -15 0>, angularTarget=<1 0 0>}
      localJointOffset=<5 0 0>
     }]
    relativeSpaceBone {boneName=pelvis}
   }
   n37 = Anim Dynamics - Bone: dyn_C_frontDress_1!(SimulationSpace:AnimPhysSimSpaceType Enum, ComponentPose=@n36, Alpha=1, GravityOverride:Vector="0,0,-980") -> (Pose) {
    angularDampingOverride=1
    angularSpringConstant=5000
    bAngularSpring=true
    bOverrideAngularDamping=true
    bUseGravityOverride=true
    bUsePlanarLimit=false
    gravityScale=0.35
    simulationSpace=RootRelative
    alphaBoolBlend {blendOption=Linear}
    boundBone {boneName=dyn_C_frontDress_1}
    componentAppliedLinearAccClamp=<3000 3000 3000>
    componentLinearAccScale=<2 2 2>
    componentLinearVelScale=<2 2 2>
    gravityOverride {z=-980}
    physicsBodyDefinitions [{
      boundBone {boneName=dyn_C_frontDress_1}
      boxExtents=<35 35 35>
      constraintSetup {angularLimitsMax=<0 0 0>, angularLimitsMin=<0 -30 0>, angularTarget=<1 0 0>}
      localJointOffset=<5 0 0>
     }]
    relativeSpaceBone {boneName=pelvis}
   }
   n38 = Anim Dynamics - Bone: dyn_C_frontDress_2!(SimulationSpace:AnimPhysSimSpaceType Enum, ComponentPose=@n37, Alpha=1, GravityOverride:Vector="0,0,-980") -> (Pose) {
    angularDampingOverride=1
    angularSpringConstant=5000
    bAngularSpring=true
    bOverrideAngularDamping=true
    bUseGravityOverride=true
    bUsePlanarLimit=false
    gravityScale=0.35
    simulationSpace=RootRelative
    alphaBoolBlend {blendOption=Linear}
    boundBone {boneName=dyn_C_frontDress_2}
    componentAppliedLinearAccClamp=<3000 3000 3000>
    componentLinearAccScale=<2 2 2>
    componentLinearVelScale=<2 2 2>
    gravityOverride {z=-980}
    physicsBodyDefinitions [{
      boundBone {boneName=dyn_C_frontDress_2}
      boxExtents=<35 35 35>
      constraintSetup {angularLimitsMax=<0 0 0>, angularLimitsMin=<0 -30 0>, angularTarget=<1 0 0>}
      localJointOffset=<5 0 0>
     }]
    relativeSpaceBone {boneName=pelvis}
   }
   n39 = Anim Dynamics - Bone: dyn_C_frontDress_3!(SimulationSpace:AnimPhysSimSpaceType Enum, ComponentPose=@n38, Alpha=1, GravityOverride:Vector="0,0,-980") -> (Pose) {
    angularDampingOverride=1
    angularSpringConstant=5000
    bAngularSpring=true
    bOverrideAngularDamping=true
    bUseGravityOverride=true
    bUsePlanarLimit=false
    gravityScale=0.35
    simulationSpace=RootRelative
    alphaBoolBlend {blendOption=Linear}
    boundBone {boneName=dyn_C_frontDress_3}
    componentAppliedLinearAccClamp=<3000 3000 3000>
    componentLinearAccScale=<2 2 2>
    componentLinearVelScale=<2 2 2>
    gravityOverride {z=-980}
    physicsBodyDefinitions [{
      boundBone {boneName=dyn_C_frontDress_3}
      boxExtents=<35 35 35>
      constraintSetup {angularLimitsMax=<0 0 0>, angularLimitsMin=<0 -30 0>, angularTarget=<1 0 0>}
      localJointOffset=<5 0 0>
     }]
    relativeSpaceBone {boneName=pelvis}
   }
   n40 = Anim Dynamics - Bone: dyn_C_frontDress_4!(SimulationSpace:AnimPhysSimSpaceType Enum, ComponentPose=@n39, Alpha=1, GravityOverride:Vector="0,0,-980") -> (Pose) {
    angularDampingOverride=1
    angularSpringConstant=5000
    bAngularSpring=true
    bOverrideAngularDamping=true
    bUseGravityOverride=true
    bUsePlanarLimit=false
    gravityScale=0.35
    simulationSpace=RootRelative
    alphaBoolBlend {blendOption=Linear}
    boundBone {boneName=dyn_C_frontDress_4}
    componentAppliedLinearAccClamp=<3000 3000 3000>
    componentLinearAccScale=<2 2 2>
    componentLinearVelScale=<2 2 2>
    gravityOverride {z=-980}
    physicsBodyDefinitions [{
      boundBone {boneName=dyn_C_frontDress_4}
      boxExtents=<35 35 35>
      constraintSetup {angularLimitsMax=<0 0 0>, angularLimitsMin=<0 -30 0>, angularTarget=<1 0 0>}
      localJointOffset=<5 0 0>
     }]
    relativeSpaceBone {boneName=pelvis}
   }
   n41 = Anim Dynamics - Bone: dyn_C_backDress_1!(SimulationSpace:AnimPhysSimSpaceType Enum, ComponentPose=@n40, Alpha=1, GravityOverride:Vector="0,0,-980") -> (Pose) {
    angularDampingOverride=1
    angularSpringConstant=5000
    bAngularSpring=true
    bOverrideAngularDamping=true
    bUseGravityOverride=true
    bUsePlanarLimit=false
    gravityScale=0.35
    simulationSpace=RootRelative
    alphaBoolBlend {blendOption=Linear}
    boundBone {boneName=dyn_C_backDress_1}
    componentAppliedLinearAccClamp=<3000 3000 3000>
    componentLinearAccScale=<2 2 2>
    componentLinearVelScale=<2 2 2>
    gravityOverride {z=-980}
    physicsBodyDefinitions [{
      boundBone {boneName=dyn_C_backDress_1}
      boxExtents=<35 35 35>
      constraintSetup {angularLimitsMax=<0 30 0>, angularLimitsMin=<0 0 0>, angularTarget=<1 0 0>}
      localJointOffset=<-5 0 0>
     }]
    relativeSpaceBone {boneName=pelvis}
   }
   n42 = Anim Dynamics - Bone: dyn_C_backDress_2!(SimulationSpace:AnimPhysSimSpaceType Enum, ComponentPose=@n41, Alpha=1, GravityOverride:Vector="0,0,-980") -> (Pose) {
    angularDampingOverride=1
    angularSpringConstant=5000
    bAngularSpring=true
    bOverrideAngularDamping=true
    bUseGravityOverride=true
    bUsePlanarLimit=false
    gravityScale=0.35
    simulationSpace=RootRelative
    alphaBoolBlend {blendOption=Linear}
    boundBone {boneName=dyn_C_backDress_2}
    componentAppliedLinearAccClamp=<3000 3000 3000>
    componentLinearAccScale=<2 2 2>
    componentLinearVelScale=<2 2 2>
    gravityOverride {z=-980}
    physicsBodyDefinitions [{
      boundBone {boneName=dyn_C_backDress_2}
      boxExtents=<35 35 35>
      constraintSetup {angularLimitsMax=<0 30 0>, angularLimitsMin=<0 0 0>, angularTarget=<1 0 0>}
      localJointOffset=<-5 0 0>
     }]
    relativeSpaceBone {boneName=pelvis}
   }
   n43 = Anim Dynamics - Bone: dyn_C_backDress_3!(SimulationSpace:AnimPhysSimSpaceType Enum, ComponentPose=@n42, Alpha=1, GravityOverride:Vector="0,0,-980") -> (Pose) {
    angularDampingOverride=1
    angularSpringConstant=5000
    bAngularSpring=true
    bOverrideAngularDamping=true
    bUseGravityOverride=true
    bUsePlanarLimit=false
    gravityScale=0.35
    simulationSpace=RootRelative
    alphaBoolBlend {blendOption=Linear}
    boundBone {boneName=dyn_C_backDress_3}
    componentAppliedLinearAccClamp=<3000 3000 3000>
    componentLinearAccScale=<2 2 2>
    componentLinearVelScale=<2 2 2>
    gravityOverride {z=-980}
    physicsBodyDefinitions [{
      boundBone {boneName=dyn_C_backDress_3}
      boxExtents=<35 35 35>
      constraintSetup {angularLimitsMax=<0 30 0>, angularLimitsMin=<0 0 0>, angularTarget=<1 0 0>}
      localJointOffset=<-5 0 0>
     }]
    relativeSpaceBone {boneName=pelvis}
   }
   n44 = Anim Dynamics - Bone: dyn_C_backDress_4!(SimulationSpace:AnimPhysSimSpaceType Enum, ComponentPose=@n43, Alpha=1, GravityOverride:Vector="0,0,-980") -> (Pose) {
    angularDampingOverride=1
    angularSpringConstant=5000
    bAngularSpring=true
    bOverrideAngularDamping=true
    bUseGravityOverride=true
    bUsePlanarLimit=false
    gravityScale=0.35
    simulationSpace=RootRelative
    alphaBoolBlend {blendOption=Linear}
    boundBone {boneName=dyn_C_backDress_4}
    componentAppliedLinearAccClamp=<3000 3000 3000>
    componentLinearAccScale=<2 2 2>
    componentLinearVelScale=<2 2 2>
    gravityOverride {z=-980}
    physicsBodyDefinitions [{
      boundBone {boneName=dyn_C_backDress_4}
      boxExtents=<35 35 35>
      constraintSetup {angularLimitsMax=<0 30 0>, angularLimitsMin=<0 0 0>, angularTarget=<1 0 0>}
      localJointOffset=<-5 0 0>
     }]
    relativeSpaceBone {boneName=pelvis}
   }
   n45 = Anim Dynamics - Bone: dyn_R_frontDress_1!(SimulationSpace:AnimPhysSimSpaceType Enum, ComponentPose=@n44, Alpha=1, GravityOverride:Vector="0,0,-980") -> (Pose) {
    angularDampingOverride=1
    angularSpringConstant=5000
    bAngularSpring=true
    bOverrideAngularDamping=true
    bUseGravityOverride=true
    bUsePlanarLimit=false
    gravityScale=0.35
    simulationSpace=RootRelative
    alphaBoolBlend {blendOption=Linear}
    boundBone {boneName=dyn_R_frontDress_1}
    componentAppliedLinearAccClamp=<3000 3000 3000>
    componentLinearAccScale=<2 2 2>
    componentLinearVelScale=<2 2 2>
    gravityOverride {z=-980}
    physicsBodyDefinitions [{
      boundBone {boneName=dyn_R_frontDress_1}
      boxExtents=<35 35 35>
      constraintSetup {angularLimitsMax=<0 0 0>, angularLimitsMin=<0 -10 0>, angularTarget=<1 0 0>}
      localJointOffset=<-5 0 0>
     }]
    relativeSpaceBone {boneName=pelvis}
   }
   n46 = Anim Dynamics - Bone: dyn_R_frontDress_2!(SimulationSpace:AnimPhysSimSpaceType Enum, ComponentPose=@n45, Alpha=1, GravityOverride:Vector="0,0,-980") -> (Pose) {
    angularDampingOverride=1
    angularSpringConstant=5000
    bAngularSpring=true
    bOverrideAngularDamping=true
    bUseGravityOverride=true
    bUsePlanarLimit=false
    gravityScale=0.35
    simulationSpace=RootRelative
    alphaBoolBlend {blendOption=Linear}
    boundBone {boneName=dyn_R_frontDress_2}
    componentAppliedLinearAccClamp=<3000 3000 3000>
    componentLinearAccScale=<2 2 2>
    componentLinearVelScale=<2 2 2>
    gravityOverride {z=-980}
    physicsBodyDefinitions [{
      boundBone {boneName=dyn_R_frontDress_2}
      boxExtents=<35 35 35>
      constraintSetup {angularLimitsMax=<0 10 0>, angularLimitsMin=<0 -10 0>, angularTarget=<1 0 0>}
      localJointOffset=<-5 0 0>
     }]
    relativeSpaceBone {boneName=pelvis}
   }
   n47 = Anim Dynamics - Bone: dyn_R_frontDress_3!(SimulationSpace:AnimPhysSimSpaceType Enum, ComponentPose=@n46, Alpha=1, GravityOverride:Vector="0,0,-980") -> (Pose) {
    angularDampingOverride=1
    angularSpringConstant=5000
    bAngularSpring=true
    bOverrideAngularDamping=true
    bUseGravityOverride=true
    bUsePlanarLimit=false
    gravityScale=0.35
    simulationSpace=RootRelative
    alphaBoolBlend {blendOption=Linear}
    boundBone {boneName=dyn_R_frontDress_3}
    componentAppliedLinearAccClamp=<3000 3000 3000>
    componentLinearAccScale=<2 2 2>
    componentLinearVelScale=<2 2 2>
    gravityOverride {z=-980}
    physicsBodyDefinitions [{
      boundBone {boneName=dyn_R_frontDress_3}
      boxExtents=<35 35 35>
      constraintSetup {angularLimitsMax=<0 20 0>, angularLimitsMin=<0 -10 0>, angularTarget=<1 0 0>}
      localJointOffset=<-5 0 0>
     }]
    relativeSpaceBone {boneName=pelvis}
   }
   n48 = Anim Dynamics - Bone: dyn_R_frontDress_4!(SimulationSpace:AnimPhysSimSpaceType Enum, ComponentPose=@n47, Alpha=1, GravityOverride:Vector="0,0,-980") -> (Pose) {
    angularDampingOverride=1
    angularSpringConstant=5000
    bAngularSpring=true
    bOverrideAngularDamping=true
    bUseGravityOverride=true
    bUsePlanarLimit=false
    gravityScale=0.35
    simulationSpace=RootRelative
    alphaBoolBlend {blendOption=Linear}
    boundBone {boneName=dyn_R_frontDress_4}
    componentAppliedLinearAccClamp=<3000 3000 3000>
    componentLinearAccScale=<2 2 2>
    componentLinearVelScale=<2 2 2>
    gravityOverride {z=-980}
    physicsBodyDefinitions [{
      boundBone {boneName=dyn_R_frontDress_4}
      boxExtents=<35 35 35>
      constraintSetup {angularLimitsMax=<0 30 0>, angularLimitsMin=<0 -10 0>, angularTarget=<1 0 0>}
      localJointOffset=<-5 0 0>
     }]
    relativeSpaceBone {boneName=pelvis}
   }
   n49 = Anim Dynamics - Bone: dyn_R_sideDress_1!(SimulationSpace:AnimPhysSimSpaceType Enum, ComponentPose=@n48, Alpha=1, GravityOverride:Vector="0,0,-980") -> (Pose) {
    angularDampingOverride=1
    angularSpringConstant=5000
    bAngularSpring=true
    bOverrideAngularDamping=true
    bUseGravityOverride=true
    bUsePlanarLimit=false
    gravityScale=0.35
    simulationSpace=RootRelative
    alphaBoolBlend {blendOption=Linear}
    boundBone {boneName=dyn_R_sideDress_1}
    componentAppliedLinearAccClamp=<3000 3000 3000>
    componentLinearAccScale=<2 2 2>
    componentLinearVelScale=<2 2 2>
    gravityOverride {z=-980}
    physicsBodyDefinitions [{
      boundBone {boneName=dyn_R_sideDress_1}
      boxExtents=<35 35 35>
      constraintSetup {angularLimitsMax=<0 0 0>, angularLimitsMin=<0 -15 0>, angularTarget=<1 0 0>}
      localJointOffset=<-5 0 0>
     }]
    relativeSpaceBone {boneName=pelvis}
   }
   n50 = Anim Dynamics - Bone: dyn_R_sideDress_2!(SimulationSpace:AnimPhysSimSpaceType Enum, ComponentPose=@n49, Alpha=1, GravityOverride:Vector="0,0,-980") -> (Pose) {
    angularDampingOverride=1
    angularSpringConstant=5000
    bAngularSpring=true
    bOverrideAngularDamping=true
    bUseGravityOverride=true
    bUsePlanarLimit=false
    gravityScale=0.35
    simulationSpace=RootRelative
    alphaBoolBlend {blendOption=Linear}
    boundBone {boneName=dyn_R_sideDress_2}
    componentAppliedLinearAccClamp=<3000 3000 3000>
    componentLinearAccScale=<2 2 2>
    componentLinearVelScale=<2 2 2>
    gravityOverride {z=-980}
    physicsBodyDefinitions [{
      boundBone {boneName=dyn_R_sideDress_2}
      boxExtents=<35 35 35>
      constraintSetup {angularLimitsMax=<0 0 0>, angularLimitsMin=<0 -15 0>, angularTarget=<1 0 0>}
      localJointOffset=<-5 0 0>
     }]
    relativeSpaceBone {boneName=pelvis}
   }
   n51 = Anim Dynamics - Bone: dyn_R_sideDress_3!(SimulationSpace:AnimPhysSimSpaceType Enum, ComponentPose=@n50, Alpha=1, GravityOverride:Vector="0,0,-980") -> (Pose) {
    angularDampingOverride=1
    angularSpringConstant=5000
    bAngularSpring=true
    bOverrideAngularDamping=true
    bUseGravityOverride=true
    bUsePlanarLimit=false
    gravityScale=0.35
    simulationSpace=RootRelative
    alphaBoolBlend {blendOption=Linear}
    boundBone {boneName=dyn_R_sideDress_3}
    componentAppliedLinearAccClamp=<3000 3000 3000>
    componentLinearAccScale=<2 2 2>
    componentLinearVelScale=<2 2 2>
    gravityOverride {z=-980}
    physicsBodyDefinitions [{
      boundBone {boneName=dyn_R_sideDress_3}
      boxExtents=<35 35 35>
      constraintSetup {angularLimitsMax=<0 0 0>, angularLimitsMin=<0 -15 0>, angularTarget=<1 0 0>}
      localJointOffset=<-5 0 0>
     }]
    relativeSpaceBone {boneName=pelvis}
   }
   n52 = Anim Dynamics - Bone: dyn_R_backDress_1!(SimulationSpace:AnimPhysSimSpaceType Enum, ComponentPose=@n51, Alpha=1, GravityOverride:Vector="0,0,-980") -> (Pose) {
    angularDampingOverride=1
    angularSpringConstant=5000
    bAngularSpring=true
    bOverrideAngularDamping=true
    bUseGravityOverride=true
    bUsePlanarLimit=false
    gravityScale=0.35
    simulationSpace=RootRelative
    alphaBoolBlend {blendOption=Linear}
    boundBone {boneName=dyn_R_backDress_1}
    componentAppliedLinearAccClamp=<3000 3000 3000>
    componentLinearAccScale=<2 2 2>
    componentLinearVelScale=<2 2 2>
    gravityOverride {z=-980}
    physicsBodyDefinitions [{
      boundBone {boneName=dyn_R_backDress_1}
      boxExtents=<35 35 35>
      constraintSetup {angularLimitsMax=<0 0 0>, angularLimitsMin=<0 -15 0>, angularTarget=<1 0 0>}
      localJointOffset=<-5 0 0>
     }]
    relativeSpaceBone {boneName=pelvis}
   }
   n53 = Anim Dynamics - Bone: dyn_R_backDress_2!(SimulationSpace:AnimPhysSimSpaceType Enum, ComponentPose=@n52, Alpha=1, GravityOverride:Vector="0,0,-980") -> (Pose) {
    angularDampingOverride=1
    angularSpringConstant=5000
    bAngularSpring=true
    bOverrideAngularDamping=true
    bUseGravityOverride=true
    bUsePlanarLimit=false
    gravityScale=0.35
    simulationSpace=RootRelative
    alphaBoolBlend {blendOption=Linear}
    boundBone {boneName=dyn_R_backDress_2}
    componentAppliedLinearAccClamp=<3000 3000 3000>
    componentLinearAccScale=<2 2 2>
    componentLinearVelScale=<2 2 2>
    gravityOverride {z=-980}
    physicsBodyDefinitions [{
      boundBone {boneName=dyn_R_backDress_2}
      boxExtents=<35 35 35>
      constraintSetup {angularLimitsMax=<0 0 0>, angularLimitsMin=<0 -15 0>, angularTarget=<1 0 0>}
      localJointOffset=<-5 0 0>
     }]
    relativeSpaceBone {boneName=pelvis}
   }
   n54 = Anim Dynamics - Bone: dyn_R_backDress_3!(SimulationSpace:AnimPhysSimSpaceType Enum, ComponentPose=@n53, Alpha=1, GravityOverride:Vector="0,0,-980") -> (Pose) {
    angularDampingOverride=1
    angularSpringConstant=5000
    bAngularSpring=true
    bOverrideAngularDamping=true
    bUseGravityOverride=true
    bUsePlanarLimit=false
    gravityScale=0.35
    simulationSpace=RootRelative
    alphaBoolBlend {blendOption=Linear}
    boundBone {boneName=dyn_R_backDress_3}
    componentAppliedLinearAccClamp=<3000 3000 3000>
    componentLinearAccScale=<2 2 2>
    componentLinearVelScale=<2 2 2>
    gravityOverride {z=-980}
    physicsBodyDefinitions [{
      boundBone {boneName=dyn_R_backDress_3}
      boxExtents=<35 35 35>
      constraintSetup {angularLimitsMax=<0 0 0>, angularLimitsMin=<0 -15 0>, angularTarget=<1 0 0>}
      localJointOffset=<-5 0 0>
     }]
    relativeSpaceBone {boneName=pelvis}
   }
   n55 = Anim Dynamics - Bone: dyn_R_backDress_4!(SimulationSpace:AnimPhysSimSpaceType Enum, ComponentPose=@n54, Alpha=1, GravityOverride:Vector="0,0,-980") -> (Pose) {
    angularDampingOverride=1
    angularSpringConstant=5000
    bAngularSpring=true
    bOverrideAngularDamping=true
    bUseGravityOverride=true
    bUsePlanarLimit=false
    gravityScale=0.35
    simulationSpace=RootRelative
    alphaBoolBlend {blendOption=Linear}
    boundBone {boneName=dyn_R_backDress_4}
    componentAppliedLinearAccClamp=<3000 3000 3000>
    componentLinearAccScale=<2 2 2>
    componentLinearVelScale=<2 2 2>
    gravityOverride {z=-980}
    physicsBodyDefinitions [{
      boundBone {boneName=dyn_R_backDress_4}
      boxExtents=<35 35 35>
      constraintSetup {angularLimitsMax=<0 0 0>, angularLimitsMin=<0 -15 0>, angularTarget=<1 0 0>}
      localJointOffset=<-5 0 0>
     }]
    relativeSpaceBone {boneName=pelvis}
   }
   n56 = Anim Dynamics - Bone: dyn_cape_01!(SimulationSpace:AnimPhysSimSpaceType Enum=BoneRelative, ComponentPose=@n55, Alpha=1, GravityOverride:Vector="0,0,-980") -> (Pose) {
    angularDampingOverride=1
    angularSpringConstant=12000
    bAngularSpring=true
    bOverrideAngularDamping=true
    bUseGravityOverride=true
    bUsePlanarLimit=false
    gravityScale=0.35
    simulationSpace=BoneRelative
    alphaBoolBlend {blendOption=Linear}
    boundBone {boneName=dyn_cape_01}
    componentAppliedLinearAccClamp=<3000 3000 3000>
    componentLinearAccScale=<2 2 2>
    componentLinearVelScale=<2 2 2>
    gravityOverride {z=-980}
    physicsBodyDefinitions [{
      boundBone {boneName=dyn_cape_01}
      boxExtents=<35 35 35>
      constraintSetup {angularLimitsMax=<0 5 15>, angularLimitsMin=<0 -40 -15>, angularTarget=<1 0 0>}
      localJointOffset=<5 0 0>
     }]
    relativeSpaceBone {boneName=pelvis}
   }
   n57 = Anim Dynamics - Bone: dyn_cape_03!(SimulationSpace:AnimPhysSimSpaceType Enum=BoneRelative, ComponentPose=@n56, Alpha=1, GravityOverride:Vector="0,0,-980") -> (Pose) {
    angularDampingOverride=1
    angularSpringConstant=12000
    bAngularSpring=true
    bOverrideAngularDamping=true
    bUseGravityOverride=true
    bUsePlanarLimit=false
    gravityScale=0.35
    simulationSpace=BoneRelative
    alphaBoolBlend {blendOption=Linear}
    boundBone {boneName=dyn_cape_03}
    componentAppliedLinearAccClamp=<3000 3000 3000>
    componentLinearAccScale=<2 2 2>
    componentLinearVelScale=<2 2 2>
    gravityOverride {z=-980}
    physicsBodyDefinitions [{
      boundBone {boneName=dyn_cape_03}
      boxExtents=<35 35 35>
      constraintSetup {angularLimitsMax=<0 5 15>, angularLimitsMin=<0 -40 -15>, angularTarget=<1 0 0>}
      localJointOffset=<5 0 0>
     }]
    relativeSpaceBone {boneName=pelvis}
   }
   n58 = Anim Dynamics - Bone: dyn_cape_05!(SimulationSpace:AnimPhysSimSpaceType Enum=BoneRelative, ComponentPose=@n57, Alpha=1, GravityOverride:Vector="0,0,-980") -> (Pose) {
    angularDampingOverride=1
    angularSpringConstant=12000
    bAngularSpring=true
    bOverrideAngularDamping=true
    bUseGravityOverride=true
    bUsePlanarLimit=false
    gravityScale=0.35
    simulationSpace=BoneRelative
    alphaBoolBlend {blendOption=Linear}
    boundBone {boneName=dyn_cape_05}
    componentAppliedLinearAccClamp=<3000 3000 3000>
    componentLinearAccScale=<2 2 2>
    componentLinearVelScale=<2 2 2>
    gravityOverride {z=-980}
    physicsBodyDefinitions [{
      boundBone {boneName=dyn_cape_05}
      boxExtents=<35 35 35>
      constraintSetup {angularLimitsMax=<0 5 15>, angularLimitsMin=<0 -40 -15>, angularTarget=<1 0 0>}
      localJointOffset=<5 0 0>
     }]
    relativeSpaceBone {boneName=pelvis}
   }
   n59 = Anim Dynamics - Bone: dyn_cape_06!(SimulationSpace:AnimPhysSimSpaceType Enum=BoneRelative, ComponentPose=@n58, Alpha=1, GravityOverride:Vector="0,0,-980") -> (Pose) {
    angularDampingOverride=1
    angularSpringConstant=12000
    bAngularSpring=true
    bOverrideAngularDamping=true
    bUseGravityOverride=true
    bUsePlanarLimit=false
    gravityScale=0.35
    simulationSpace=BoneRelative
    alphaBoolBlend {blendOption=Linear}
    boundBone {boneName=dyn_cape_06}
    componentAppliedLinearAccClamp=<3000 3000 3000>
    componentLinearAccScale=<2 2 2>
    componentLinearVelScale=<2 2 2>
    gravityOverride {z=-980}
    physicsBodyDefinitions [{
      boundBone {boneName=dyn_cape_06}
      boxExtents=<35 35 35>
      constraintSetup {angularLimitsMax=<0 5 15>, angularLimitsMin=<0 -40 -15>, angularTarget=<1 0 0>}
      localJointOffset=<5 0 0>
     }]
    relativeSpaceBone {boneName=pelvis}
   }
   n60 = Anim Dynamics - Bone: dyn_cape_r_01!(SimulationSpace:AnimPhysSimSpaceType Enum=BoneRelative, ComponentPose=@n59, Alpha=1, GravityOverride:Vector="0,0,-980") -> (Pose) {
    angularDampingOverride=1
    angularSpringConstant=12000
    bAngularSpring=true
    bOverrideAngularDamping=true
    bUseGravityOverride=true
    bUsePlanarLimit=false
    gravityScale=0.35
    simulationSpace=BoneRelative
    alphaBoolBlend {blendOption=Linear}
    boundBone {boneName=dyn_cape_r_01}
    componentAppliedLinearAccClamp=<3000 3000 3000>
    componentLinearAccScale=<2 2 2>
    componentLinearVelScale=<2 2 2>
    gravityOverride {z=-980}
    physicsBodyDefinitions [{
      boundBone {boneName=dyn_cape_r_01}
      boxExtents=<35 35 35>
      constraintSetup {angularLimitsMax=<0 5 15>, angularLimitsMin=<0 -40 -15>, angularTarget=<1 0 0>}
      localJointOffset=<5 0 0>
     }]
    relativeSpaceBone {boneName=pelvis}
   }
   n61 = Anim Dynamics - Bone: dyn_cape_r_03!(SimulationSpace:AnimPhysSimSpaceType Enum=BoneRelative, ComponentPose=@n60, Alpha=1, GravityOverride:Vector="0,0,-980") -> (Pose) {
    angularDampingOverride=1
    angularSpringConstant=12000
    bAngularSpring=true
    bOverrideAngularDamping=true
    bUseGravityOverride=true
    bUsePlanarLimit=false
    gravityScale=0.35
    simulationSpace=BoneRelative
    alphaBoolBlend {blendOption=Linear}
    boundBone {boneName=dyn_cape_r_03}
    componentAppliedLinearAccClamp=<3000 3000 3000>
    componentLinearAccScale=<2 2 2>
    componentLinearVelScale=<2 2 2>
    gravityOverride {z=-980}
    physicsBodyDefinitions [{
      boundBone {boneName=dyn_cape_r_03}
      boxExtents=<35 35 35>
      constraintSetup {angularLimitsMax=<0 5 15>, angularLimitsMin=<0 -40 -15>, angularTarget=<1 0 0>}
      localJointOffset=<5 0 0>
     }]
    relativeSpaceBone {boneName=pelvis}
   }
   n62 = Anim Dynamics - Bone: dyn_cape_r_05!(SimulationSpace:AnimPhysSimSpaceType Enum=BoneRelative, ComponentPose=@n61, Alpha=1, GravityOverride:Vector="0,0,-980") -> (Pose) {
    angularDampingOverride=1
    angularSpringConstant=12000
    bAngularSpring=true
    bOverrideAngularDamping=true
    bUseGravityOverride=true
    bUsePlanarLimit=false
    gravityScale=0.35
    simulationSpace=BoneRelative
    alphaBoolBlend {blendOption=Linear}
    boundBone {boneName=dyn_cape_r_05}
    componentAppliedLinearAccClamp=<3000 3000 3000>
    componentLinearAccScale=<2 2 2>
    componentLinearVelScale=<2 2 2>
    gravityOverride {z=-980}
    physicsBodyDefinitions [{
      boundBone {boneName=dyn_cape_r_05}
      boxExtents=<35 35 35>
      constraintSetup {angularLimitsMax=<0 5 15>, angularLimitsMin=<0 -40 -15>, angularTarget=<1 0 0>}
      localJointOffset=<5 0 0>
     }]
    relativeSpaceBone {boneName=pelvis}
   }
   n63 = Anim Dynamics - Bone: dyn_cape_r_06!(SimulationSpace:AnimPhysSimSpaceType Enum=BoneRelative, ComponentPose=@n62, Alpha=1, GravityOverride:Vector="0,0,-980") -> (Pose) {
    angularDampingOverride=1
    angularSpringConstant=12000
    bAngularSpring=true
    bOverrideAngularDamping=true
    bUseGravityOverride=true
    bUsePlanarLimit=false
    gravityScale=0.35
    simulationSpace=BoneRelative
    alphaBoolBlend {blendOption=Linear}
    boundBone {boneName=dyn_cape_r_06}
    componentAppliedLinearAccClamp=<3000 3000 3000>
    componentLinearAccScale=<2 2 2>
    componentLinearVelScale=<2 2 2>
    gravityOverride {z=-980}
    physicsBodyDefinitions [{
      boundBone {boneName=dyn_cape_r_06}
      boxExtents=<35 35 35>
      constraintSetup {angularLimitsMax=<0 5 15>, angularLimitsMin=<0 -40 -15>, angularTarget=<1 0 0>}
      localJointOffset=<5 0 0>
     }]
    relativeSpaceBone {boneName=pelvis}
   }
   n64 = Anim Dynamics - Bone: dyn_cape_l_01!(SimulationSpace:AnimPhysSimSpaceType Enum=BoneRelative, ComponentPose=@n63, Alpha=1, GravityOverride:Vector="0,0,-980") -> (Pose) {
    angularDampingOverride=1
    angularSpringConstant=12000
    bAngularSpring=true
    bOverrideAngularDamping=true
    bUseGravityOverride=true
    bUsePlanarLimit=false
    gravityScale=0.35
    simulationSpace=BoneRelative
    alphaBoolBlend {blendOption=Linear}
    boundBone {boneName=dyn_cape_l_01}
    componentAppliedLinearAccClamp=<3000 3000 3000>
    componentLinearAccScale=<2 2 2>
    componentLinearVelScale=<2 2 2>
    gravityOverride {z=-980}
    physicsBodyDefinitions [{
      boundBone {boneName=dyn_cape_l_01}
      boxExtents=<35 35 35>
      constraintSetup {angularLimitsMax=<0 5 15>, angularLimitsMin=<0 -40 -15>, angularTarget=<1 0 0>}
      localJointOffset=<5 0 0>
     }]
    relativeSpaceBone {boneName=pelvis}
   }
   n65 = Anim Dynamics - Bone: dyn_cape_l_03!(SimulationSpace:AnimPhysSimSpaceType Enum=BoneRelative, ComponentPose=@n64, Alpha=1, GravityOverride:Vector="0,0,-980") -> (Pose) {
    angularDampingOverride=1
    angularSpringConstant=12000
    bAngularSpring=true
    bOverrideAngularDamping=true
    bUseGravityOverride=true
    bUsePlanarLimit=false
    gravityScale=0.35
    simulationSpace=BoneRelative
    alphaBoolBlend {blendOption=Linear}
    boundBone {boneName=dyn_cape_l_03}
    componentAppliedLinearAccClamp=<3000 3000 3000>
    componentLinearAccScale=<2 2 2>
    componentLinearVelScale=<2 2 2>
    gravityOverride {z=-980}
    physicsBodyDefinitions [{
      boundBone {boneName=dyn_cape_l_03}
      boxExtents=<35 35 35>
      constraintSetup {angularLimitsMax=<0 5 15>, angularLimitsMin=<0 -40 -15>, angularTarget=<1 0 0>}
      localJointOffset=<5 0 0>
     }]
    relativeSpaceBone {boneName=pelvis}
   }
   n66 = Anim Dynamics - Bone: dyn_cape_l_05!(SimulationSpace:AnimPhysSimSpaceType Enum=BoneRelative, ComponentPose=@n65, Alpha=1, GravityOverride:Vector="0,0,-980") -> (Pose) {
    angularDampingOverride=1
    angularSpringConstant=12000
    bAngularSpring=true
    bOverrideAngularDamping=true
    bUseGravityOverride=true
    bUsePlanarLimit=false
    gravityScale=0.35
    simulationSpace=BoneRelative
    alphaBoolBlend {blendOption=Linear}
    boundBone {boneName=dyn_cape_l_05}
    componentAppliedLinearAccClamp=<3000 3000 3000>
    componentLinearAccScale=<2 2 2>
    componentLinearVelScale=<2 2 2>
    gravityOverride {z=-980}
    physicsBodyDefinitions [{
      boundBone {boneName=dyn_cape_l_05}
      boxExtents=<35 35 35>
      constraintSetup {angularLimitsMax=<0 5 15>, angularLimitsMin=<0 -40 -15>, angularTarget=<1 0 0>}
      localJointOffset=<5 0 0>
     }]
    relativeSpaceBone {boneName=pelvis}
   }
   n67 = Anim Dynamics - Bone: dyn_cape_l_06!(SimulationSpace:AnimPhysSimSpaceType Enum=BoneRelative, ComponentPose=@n66, Alpha=1, GravityOverride:Vector="0,0,-980") -> (Pose) {
    angularDampingOverride=1
    angularSpringConstant=12000
    bAngularSpring=true
    bOverrideAngularDamping=true
    bUseGravityOverride=true
    bUsePlanarLimit=false
    gravityScale=0.35
    simulationSpace=BoneRelative
    alphaBoolBlend {blendOption=Linear}
    boundBone {boneName=dyn_cape_l_06}
    componentAppliedLinearAccClamp=<3000 3000 3000>
    componentLinearAccScale=<2 2 2>
    componentLinearVelScale=<2 2 2>
    gravityOverride {z=-980}
    physicsBodyDefinitions [{
      boundBone {boneName=dyn_cape_l_06}
      boxExtents=<35 35 35>
      constraintSetup {angularLimitsMax=<0 5 15>, angularLimitsMin=<0 -40 -15>, angularTarget=<1 0 0>}
      localJointOffset=<5 0 0>
     }]
    relativeSpaceBone {boneName=pelvis}
   }
   n68 = "Transform (Modify) Bone - Bone: dyn_cape_l_01"!(ComponentPose=@n67, bAlphaBoolEnabled:bool=True) -> (Pose) {
    alphaInputType=Bool
    rotationMode=BMM_Additive
    rotationSpace=BCS_BoneSpace
    alphaBoolBlend {blendOption=Linear}
    boneToModify {boneName=dyn_cape_l_01}
    rotation {pitch=35}
   }
   n69 = "Transform (Modify) Bone - Bone: dyn_cape_01"!(ComponentPose=@n68, bAlphaBoolEnabled:bool=True) -> (Pose) {
    alphaInputType=Bool
    rotationMode=BMM_Additive
    rotationSpace=BCS_BoneSpace
    alphaBoolBlend {blendOption=Linear}
    boneToModify {boneName=dyn_cape_01}
    rotation {pitch=35}
   }
   n70 = "Transform (Modify) Bone - Bone: dyn_cape_r_01"!(ComponentPose=@n69, bAlphaBoolEnabled:bool=True) -> (Pose) {
    alphaInputType=Bool
    rotationMode=BMM_Additive
    rotationSpace=BCS_BoneSpace
    alphaBoolBlend {blendOption=Linear}
    boneToModify {boneName=dyn_cape_r_01}
    rotation {pitch=35}
   }
   n71 = "Transform (Modify) Bone - Bone: dyn_cape_l_04"!(ComponentPose=@n70, bAlphaBoolEnabled:bool=True) -> (Pose) {
    alphaInputType=Bool
    rotationMode=BMM_Additive
    rotationSpace=BCS_BoneSpace
    alphaBoolBlend {blendOption=Linear}
    boneToModify {boneName=dyn_cape_l_04}
    rotation {pitch=25}
   }
   n72 = "Transform (Modify) Bone - Bone: dyn_cape_04"!(ComponentPose=@n71, bAlphaBoolEnabled:bool=True) -> (Pose) {
    alphaInputType=Bool
    rotationMode=BMM_Additive
    rotationSpace=BCS_BoneSpace
    alphaBoolBlend {blendOption=Linear}
    boneToModify {boneName=dyn_cape_04}
    rotation {pitch=25}
   }
   n73 = "Transform (Modify) Bone - Bone: dyn_cape_r_04"!(ComponentPose=@n72, bAlphaBoolEnabled:bool=True) -> (Pose) {
    alphaInputType=Bool
    rotationMode=BMM_Additive
    rotationSpace=BCS_BoneSpace
    alphaBoolBlend {blendOption=Linear}
    boneToModify {boneName=dyn_cape_r_04}
    rotation {pitch=25}
   }
   n74 = "Transform (Modify) Bone - Bone: dyn_cape_l_06"!(ComponentPose=@n73, bAlphaBoolEnabled:bool=True) -> (Pose) {
    alphaInputType=Bool
    rotationMode=BMM_Additive
    rotationSpace=BCS_BoneSpace
    alphaBoolBlend {blendOption=Linear}
    boneToModify {boneName=dyn_cape_l_06}
    rotation {pitch=15}
   }
   n75 = "Transform (Modify) Bone - Bone: dyn_cape_06"!(ComponentPose=@n74, bAlphaBoolEnabled:bool=True) -> (Pose) {
    alphaInputType=Bool
    rotationMode=BMM_Additive
    rotationSpace=BCS_BoneSpace
    alphaBoolBlend {blendOption=Linear}
    boneToModify {boneName=dyn_cape_06}
    rotation {pitch=15}
   }
   n76 = "Transform (Modify) Bone - Bone: dyn_cape_r_06"!(ComponentPose=@n75, bAlphaBoolEnabled:bool=True) -> (Pose) {
    alphaInputType=Bool
    rotationMode=BMM_Additive
    rotationSpace=BCS_BoneSpace
    alphaBoolBlend {blendOption=Linear}
    boneToModify {boneName=dyn_cape_r_06}
    rotation {pitch=15}
   }
   n77 = Trail controller - Bone: dyn_cape_07(ComponentPose=@n76, Alpha=1) {
    bLimitRotation=true
    bShowBaseMotion=false
    chainLength=7
    maxDeltaTime=0.3
    alphaScaleBiasClamp {bInterpResult=true, scale=0.75}
    rotationLimits [
     {limitMax=<45 1 90>, limitMin=<-45 -1 -90>}
     {limitMax=<45 1 90>, limitMin=<-45 -1 -90>}
     {limitMax=<45 1 90>, limitMin=<-45 -1 -90>}
     {limitMax=<45 1 90>, limitMin=<-45 -1 -90>}
     {limitMax=<45 1 90>, limitMin=<-45 -1 -90>}
     {limitMax=<45 1 90>, limitMin=<-45 -1 -90>}
     {limitMax=<45 1 90>, limitMin=<-45 -1 -90>}
    ]
    rotationOffsets [<0 0 0>, <0 0 0>, <0 0 0>, <0 0 0>, <0 0 0>, <0 0 0>, <0 0 0>]
    trailBone {boneName=dyn_cape_07}
    trailRelaxationSpeed {editorCurveData {keys [<at 0 10 RCIM_Linear>,<at 0.5963 5.9424 RCIM_Linear>,<at 1 5 RCIM_Linear>]}}
   }
   n78 = Trail controller - Bone: dyn_cape_r_07(ComponentPose=@n77, Alpha=1) {
    bLimitRotation=true
    bShowBaseMotion=false
    chainLength=7
    maxDeltaTime=0.3
    alphaScaleBiasClamp {bInterpResult=true, scale=0.75}
    rotationLimits [
     {limitMax=<45 1 90>, limitMin=<-45 -1 -90>}
     {limitMax=<45 1 90>, limitMin=<-45 -1 -90>}
     {limitMax=<45 1 90>, limitMin=<-45 -1 -90>}
     {limitMax=<45 1 90>, limitMin=<-45 -1 -90>}
     {limitMax=<45 1 90>, limitMin=<-45 -1 -90>}
     {limitMax=<45 1 90>, limitMin=<-45 -1 -90>}
     {limitMax=<45 1 90>, limitMin=<-45 -1 -90>}
    ]
    rotationOffsets [<0 0 0>, <0 0 0>, <0 0 0>, <0 0 0>, <0 0 0>, <0 0 0>, <0 0 0>]
    trailBone {boneName=dyn_cape_r_07}
    trailRelaxationSpeed {editorCurveData {keys [<at 0 10 RCIM_Linear>,<at 0.5963 5.8564 RCIM_Linear>,<at 1 5 RCIM_Linear>]}}
   }
   n79 = Trail controller - Bone: dyn_cape_l_07(ComponentPose=@n78, Alpha=1) {
    bLimitRotation=true
    bShowBaseMotion=false
    chainLength=7
    maxDeltaTime=0.3
    alphaScaleBiasClamp {bInterpResult=true, scale=0.75}
    rotationLimits [
     {limitMax=<45 1 90>, limitMin=<-45 -1 -90>}
     {limitMax=<45 1 90>, limitMin=<-45 -1 -90>}
     {limitMax=<45 1 90>, limitMin=<-45 -1 -90>}
     {limitMax=<45 1 90>, limitMin=<-45 -1 -90>}
     {limitMax=<45 1 90>, limitMin=<-45 -1 -90>}
     {limitMax=<45 1 90>, limitMin=<-45 -1 -90>}
     {limitMax=<45 1 90>, limitMin=<-45 -1 -90>}
    ]
    rotationOffsets [<0 0 0>, <0 0 0>, <0 0 0>, <0 0 0>, <0 0 0>, <0 0 0>, <0 0 0>]
    trailBone {boneName=dyn_cape_l_07}
    trailRelaxationSpeed {editorCurveData {keys [<at 0 10 RCIM_Linear>,<at 0.6109 5.9854 RCIM_Linear>,<at 1 5 RCIM_Linear>]}}
   }
   Save cached pose 'animDyn Cache'(Pose=Component To Local(ComponentPose=@n79)) {cacheName=animDyn Cache}
   n80 = RigidBody!(ComponentPose=Local To Component(LocalPose=Use cached pose 'Start_cache'()), Alpha, OverrideWorldGravity:Vector="0,0,-980") -> (Pose) {
    bClampLinearTranslationLimitToRefPose=true
    bOverrideWorldGravity=true
    overridePhysicsAsset=&/F_MED_ForwardLake_DireLab_DressPhysics
    simulationSpace=BaseBoneSpace
    alphaBoolBlend {blendOption=Linear}
    baseBoneRef {boneName=root}
    overrideWorldGravity {z=-980}
    simSpaceSettings {
     dampingAlpha=0
     maxAngularAcceleration=30
     maxAngularVelocity=3
     maxLinearAcceleration=1500
     maxLinearVelocity=600
     velocityScaleZ=0.1
     worldAlpha=0.8
     externalLinearDragV=<3 3 3>
    }
   }
   n81 = Component To Local(ComponentPose=@n80)
   n82 = RigidBody(ComponentPose=Local To Component(LocalPose=Use cached pose 'Start_cache'()), OverrideWorldGravity="0,0,-980") {
    bClampLinearTranslationLimitToRefPose=true
    bOverrideWorldGravity=true
    overridePhysicsAsset=&/F_MED_ForwardLake_DireLab_CapeBackblingPhysics
    simulationSpace=BaseBoneSpace
    alphaBoolBlend {blendOption=Linear}
    baseBoneRef {boneName=root}
    overrideWorldGravity {z=-980}
    simSpaceSettings {
     dampingAlpha=0
     maxAngularAcceleration=30
     maxAngularVelocity=3
     maxLinearAcceleration=1500
     maxLinearVelocity=600
     velocityScaleZ=0.35
     worldAlpha=1
     externalLinearDragV=<3 3 3>
    }
   }
   n83 = Component To Local(ComponentPose=@n82)
   n84 = Get Is Wearing Backbling()
   n85 = "Transform (Modify) Bone - Bone: dyn_cape_l_MSTR"!(
    ComponentPose=Local To Component(LocalPose=Blend Poses by bool(BlendPose_0=@n83, BlendPose_1=@n81, bActiveValue=@n84) {blendType=Linear})
    Alpha=1
   ) -> (Pose) {
    rotationMode=BMM_Additive
    rotationSpace=BCS_BoneSpace
    alphaBoolBlend {blendOption=Linear}
    alphaScaleBiasClamp {bInterpResult=true, interpSpeedDecreasing=20, interpSpeedIncreasing=20}
    boneToModify {boneName=dyn_cape_l_MSTR}
    rotation {pitch=15}
   }
   n86 = "Transform (Modify) Bone - Bone: dyn_cape_MSTR"!(ComponentPose=@n85, Alpha=1) -> (Pose) {
    rotationMode=BMM_Additive
    rotationSpace=BCS_BoneSpace
    alphaBoolBlend {blendOption=Linear}
    alphaScaleBiasClamp {bInterpResult=true, interpSpeedDecreasing=20, interpSpeedIncreasing=20}
    boneToModify {boneName=dyn_cape_MSTR}
    rotation {pitch=15}
   }
   n87 = "Transform (Modify) Bone - Bone: dyn_cape_r_MSTR"!(ComponentPose=@n86, Alpha=1) -> (Pose) {
    rotationMode=BMM_Additive
    rotationSpace=BCS_BoneSpace
    alphaBoolBlend {blendOption=Linear}
    alphaScaleBiasClamp {bInterpResult=true, interpSpeedDecreasing=20, interpSpeedIncreasing=20}
    boneToModify {boneName=dyn_cape_r_MSTR}
    rotation {pitch=15}
   }
   Save cached pose 'Full_RBAN_cache'(Pose=Component To Local(ComponentPose=@n87)) {cacheName=Full_RBAN_cache}
   n88 = RigidBody(ComponentPose=Local To Component(LocalPose=Use cached pose 'Start_cache'()), OverrideWorldGravity="0,0,-980") {
    bClampLinearTranslationLimitToRefPose=true
    bOverrideWorldGravity=true
    overridePhysicsAsset=&/F_MED_ForwardLake_DireLab_DressOnlyPhysics
    simulationSpace=BaseBoneSpace
    alphaBoolBlend {blendOption=Linear}
    baseBoneRef {boneName=root}
    overrideWorldGravity {z=-980}
    simSpaceSettings {
     dampingAlpha=0
     maxAngularAcceleration=30
     maxAngularVelocity=3
     maxLinearAcceleration=1500
     maxLinearVelocity=600
     velocityScaleZ=0.1
     worldAlpha=0.8
     externalLinearDragV=<3 3 3>
    }
   }
   Save cached pose 'onlyDress RBAN_cache'(Pose=Component To Local(ComponentPose=@n88)) {cacheName=onlyDress RBAN_cache}
  ]}
 EventGraph {nodes [
   n1 = Event BlueprintUpdateAnimation()
   > n2 = Sequence!(execute) -> (then_0, then_1, then_2:Exec, then_3:Exec, then_4:Exec, then_5:Exec)
   > Branch(Condition=Get bClothEnabled())
   > n3 = Sequence!(execute) -> (then_0, then_1, then_2)
   n4 = GetClothingSimulationInteractor!(self:Skeletal Mesh Component Object Reference=GetOwningComponent(), ClothingAssetName:Name=None) -> (ReturnValue:Clothing Simulation Interactor Object Reference)
   Is Valid(exec=@n3.then_2, InputObject=@n4)
   n5 = MakeLiteralString(Value=F_MED_ForwardLake_DireLab_Clothing_0)
   > n6 = Cast To ChaosClothingInteractor!(
    execute
    Object=GetClothingInteractor!(self:Clothing Simulation Interactor Object Reference=@n4, ClothingAssetName:Name=String To Name(InString=@n5)) -> (ReturnValue:Clothing Interactor Object Reference)
   ) -> (then, CastFailed, AsChaos Clothing Interactor:Chaos Clothing Interactor Object Reference) {is=title K2Node_DynamicCast}
   > Set ChaosClothingInteractor(ChaosClothingInteractor:Chaos Clothing Interactor Object Reference=@n6) -> (Output_Get:Chaos Clothing Interactor Object Reference)
   n7 = GetGameTimeInSeconds()
   n8 = "Cloth Controller Function (Parameter Asset) (Nativized)"!(
    execute=@n3.then_1
    CharacterPartAnimInstance:Custom Character Part Anim Instance Object Reference=Self-Reference()
    ClothingAssetName:Name=F_MED_ForwardLake_DireLab_Clothing_0
    ClothCoreSettings:"Cloth Core Settings Structure (by ref)"=Get Cloth Core Settings_0() -> (Cloth Core Settings_0:Cloth Core Settings Structure)
    WindGustLocalSettings:"Wind Gust Settings Structure (by ref)"=Get Wind Gust Local Settings() -> (Wind Gust Local Settings:Wind Gust Settings Structure)
    ClothUpAxis:Vector="1,0,0"
    LinVelScaleFwdBwdClamp:float=0
    LinVelScaleUpDwnClamp:float=0.7
    DeltaTime:float=@n1.DeltaTimeX
    GameTimeInSeconds:float=@n7
    StateLogicLayerName:Name=none
    Parameters:Fort Character Dynamics Parameters Object Reference=&/ForwardLake_DireLab_Parameters
   ) -> (
    then
    OutCurrentState:Name=None
    OutClothParameters:Cloth Parameters Structure
    OutClothCoreSettings:Cloth Core Settings Structure
    OutWindGustSettings:Wind Gust Parameters Structure
    OutWindGustLocalSettings:Wind Gust Settings Structure
    OutWindVelocity:Vector
    OutLinearVelVector:Vector
   )
   > Set Wind Gust Settings Out(Wind Gust Settings Out:Wind Gust Parameters Structure=@n8.OutWindGustSettings) -> (Output_Get:Wind Gust Parameters Structure)
   > Set Wind Gust Local Settings Out(Wind Gust Local Settings Out:Wind Gust Settings Structure=@n8.OutWindGustLocalSettings) -> (Output_Get:Wind Gust Settings Structure)
   Branch(execute=@n2.then_3, Condition=Get bIsInFrontEnd())
   > Set isPlayingFrontendIdle(isPlayingFrontendIdle:float=GetCurveValue(self=GetOwnerFrontEndAnimInstance(), CurveName=Frontend_Idle_Bypass)) -> (Output_Get:float)
   Do Once(execute=@n2.then_5)
   > Delay(Duration=0.2)
   n9 = NOT Boolean(A=Get bIsInFrontEnd())
   n10 = NOT Boolean(A=Get bIsSkyDiving())
   n11 = NOT Boolean(A=Get bIsOnGround() -> (bIsOnGround:bool))
   > Set Is in BattleBus Float(Is in BattleBus Float:double="To Float (Boolean)"(InBool=AND Boolean!(A=@n10, B=@n11, C:bool=@n9) -> (ReturnValue))) -> (Output_Get:double)
   n12 = PlayerMovement(execute=@n2.then_4, AnimInstance=Self-Reference(), XDotProductCoefficient=1, YDotProductCoefficient=1, ZDotProductCoefficient=1)
   > n13 = Branch(Condition="Equal (Float)"(A=Get isPlayingFrontendIdle()))
   > n14 = Branch(Condition=Get animDyn Off RBAN Off Cloth Off() -> (animDyn Off RBAN Off Cloth Off:bool))
   > n15 = Branch(Condition=Get bIsSkyDiving())
   Set BwdMovement(execute=[@n16.else, @n15.then], BwdMovement:double=1) -> (Output_Get:double)
   n16 = Branch(execute=@n15.else, Condition="float < float"(A=@n12.YMovement))
   Set BwdMovement(execute=[@n16.then, @n13.else, @n14.else], BwdMovement=0)
   n17 = NOT Boolean(A=Get bClothEnabled())
   n18 = Get bRigidBodyEnabled()
   Set RBAN On Cloth Off(execute=@n2.then_1, RBAN On Cloth Off:bool=OR Boolean(A=AND Boolean!(A=@n18, B=@n17) -> (ReturnValue), B=Get Is Wearing Backbling())) -> (Output_Get:bool)
   > Set RBAN On Cloth On(RBAN On Cloth On:bool=AND Boolean(A=Get bRigidBodyEnabled(), B=Get bClothEnabled())) -> (Output_Get:bool)
   > n19 = Sequence!(execute) -> (then_0, then_1)
   n20 = Get bClothEnabled()
   n21 = Get bRigidBodyEnabled()
   > n22 = Branch(Condition=AND Boolean(A=NOT Boolean(A=@n21), B=NOT Boolean(A=@n20)))
   > Set RBAN Off Cloth Off(RBAN Off Cloth Off:bool=true) -> (Output_Get:bool)
   > Set NOT RBAN Off Cloth Off(NOT RBAN Off Cloth Off:bool=false) -> (Output_Get:bool)
   Set RBAN Off Cloth Off(execute=@n22.else, RBAN Off Cloth Off=false)
   > Set NOT RBAN Off Cloth Off(NOT RBAN Off Cloth Off=true)
   n23 = Get bRigidBodyEnabled()
   n24 = Get bClothEnabled()
   n25 = Get bAnimDynamicsEnabled() -> (bAnimDynamicsEnabled:bool)
   Set animDyn Off RBAN Off Cloth Off(
    execute=@n19.then_1
    animDyn Off RBAN Off Cloth Off:bool=AND Boolean!(A=NOT Boolean(A=@n25), B=NOT Boolean(A=@n23), C:bool=NOT Boolean(A=@n24)) -> (ReturnValue)
   ) -> (Output_Get:bool)
   n26 = "Gravity Override State Function (Parameter Asset) (Nativized)"(
    execute=@n2.then_2
    CharacterPartAnimInstance=Self-Reference()
    GameTimeInSeconds=@n7
    Parameters=&ForwardLake_DireLab_Parameters
   )
   > Set Gravity Override(Gravity Override=@n26)
   > Set crouchMoving(crouchMoving:bool=AND Boolean!(A=Get bIsCrouching(), B=Get bIsAccelerating2D()) -> (ReturnValue)) -> (Output_Get:bool)
   > n27 = Branch(Condition=OR Boolean(A=Get bIsInFrontEnd(), B=Get bIsDBNO()))
   > Set simSpaceChange(simSpaceChange:AnimPhysSimSpaceType Enum=World) -> (Output_Get:AnimPhysSimSpaceType Enum)
   Set simSpaceChange(execute=@n27.else)
   n28 = Event OnBeginJumping()
   Set getCustomJump(execute=[@n28.then, @n29.then], getCustomJump:bool=true) -> (Output_Get:bool)
   n29 = Event OnBeginFalling()
   n30 = Event BlueprintInitializeAnimation()
   n31 = Sequence!(execute=[@n30.then, @n35.then]) -> (then_0, then_1, then_2:Exec, then_3:Exec)
   ResetDynamics(execute=@n31.then_3, self=Self-Reference(), InTeleportType=TeleportPhysics)
   n32 = Set Is Wearing Backbling(execute=@n31.then_1, Is Wearing Backbling=CheckForBackbling())
   n33 = Branch(execute=@n31.then_2, Condition=@n32)
   n34 = GetPartSkeletalMeshComponent()
   > Set ClothBlendWeight!(execute, ClothBlendWeight:float=0, self:Skeletal Mesh Component Object Reference=@n34) -> (then, Output_Get:float)
   > Set bDisableClothSimulation!(execute, bDisableClothSimulation:bool=true, self:Skeletal Mesh Component Object Reference=@n34) -> (then, Output_Get:bool)
   Set ClothBlendWeight(execute=@n33.else, ClothBlendWeight=1, self=@n34)
   > Set bDisableClothSimulation(bDisableClothSimulation=false, self=@n34)
   Set frontendAnimDynAlpha(execute=@n31.then_0, frontendAnimDynAlpha:double=SelectFloat(A=0.6, B=1, bPickA=Get bIsInFrontEnd())) -> (Output_Get:double)
   n35 = Event OnCharacterCustomizationLoaded()
   n36 = Event BlueprintProcessAnimInputEvent()
   > n37 = Branch(Condition=IsNamed(self=@n36.AnimInputEvent, InEventName=aie.Player.Zipline.Entered))
   > Delay(Duration=0.1)
   n38 = Get ChaosClothingInteractor() -> (ChaosClothingInteractor:Chaos Clothing Interactor Object Reference)
   > ResetAndTeleport!(execute, self:Chaos Clothing Interactor Object Reference=@n38, bReset:bool=true, bTeleport:bool=true) -> (then)
   Branch(execute=@n37.else, Condition=IsNamed(self=@n36.AnimInputEvent, InEventName=aie.Player.Zipline.Exited))
   > Set resetClothLandingFromZilpine(resetClothLandingFromZilpine:bool=true) -> (Output_Get:bool)
   Event OnEndFalling()
   > n39 = Sequence!(execute) -> (then_0, then_1)
   > Branch(Condition=Get resetClothLandingFromZilpine() -> (resetClothLandingFromZilpine:bool))
   > ResetAndTeleport(self=@n38, bReset=true, bTeleport=true)
   > Set resetClothLandingFromZilpine(resetClothLandingFromZilpine=false)
   Set getCustomJump(execute=@n39.then_1, getCustomJump=false)
  ]}
]
variables [
 animDyn Off RBAN Off Cloth Off {type=boolean}
 BwdMovement {type=number}
 ChaosClothingInteractor {type=ChaosClothingInteractor, default=null}
 Cloth Core Settings_0 {
  type=ClothCoreSettings
  default {
   collisionThickness=0.1
   dampingCoefficient=0.1
   frictionCoefficient=0.3
   localDampingCoefficient=0
   numberIterations=2
   numberSubsteps=1
   selfCollisionThickness=0
   animDriveDamping=<0.2 1>
   animDriveStiffness=<0 1>
   areaStiffness=<1 1>
   bendingStiffness=<1 1>
   edgeStiffness=<1 1>
   tetherStiffness=<1 1>
  }
 }
 crouchMoving {type=boolean}
 frontendAnimDynAlpha {type=number}
 getCustomJump {type=boolean}
 Gravity Override {type=Vector}
 Is in BattleBus Float {type=number, default=1}
 Is Wearing Backbling {type=boolean}
 isGliderSurfing {type=boolean}
 isPlayingFrontendIdle {type=number}
 NOT RBAN Off Cloth Off {type=boolean}
 RBAN Off Cloth Off {type=boolean}
 RBAN On Cloth Off {type=boolean}
 RBAN On Cloth On {type=boolean}
 resetClothLandingFromZilpine {type=boolean}
 simSpaceChange {type=AnimPhysSimSpaceType}
 Wind Gust Local Settings {type=WindGustSettings, default {elapsedTime=0, finalGustStrengthLocal=0, gustDuration=0, gustStrength=0, timeBetweenGusts=0}}
 Wind Gust Local Settings Out {type=WindGustSettings, default {elapsedTime=0, finalGustStrengthLocal=0, gustDuration=0, gustStrength=0, timeBetweenGusts=0}}
 Wind Gust Settings Out {
  type=WindGustParameters
  default {
   maxDownInterpSpeed=0
   maxGustDuration=0
   maxGustStrength=0
   maxTimeBetweenGusts=0
   maxUpInterpSpeed=0
   minDownInterpSpeed=0
   minGustDuration=0
   minGustStrength=0
   minTimeBetweenGusts=0
   minUpInterpSpeed=0
   windTurbulenceScalarA=0
   windTurbulenceScalarB=0
   windAmplitude=<0 0 0>
   windFrequency=<0 0 0>
   windOffset=<0 0 0>
  }
 }
]
