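# Self-collision physics asset for the F_MED_ForwardLake_DireLab female medium player body cosmetic.
# Single capsules cover pelvis, spine, clavicles, arms down to the forearms, and legs down to the feet, with none on head, neck or hands.
# More capsules sit on the dyn_ dress bone chains: center, left and right back, plus left and right sides.
# It defines no constraints, so it supplies collision shapes only, not a jointed ragdoll.
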PhysicsAsset /BRCosmetics/Characters/Player/Female/Medium/Bodies/F_MED_ForwardLake_DireLab/Meshes/F_MED_ForwardLake_DireLab_SelfCollision_Physics
bodies {
 calf_l [{shapeType=Capsule, length=30, radius=9, shapeName=calf_l_capsule, center=<-23.486 0 0>, rotation=<rot 90 0 0>}]
 calf_r [{shapeType=Capsule, length=30, radius=9, shapeName=calf_r_capsule, center=<23.486 0 0>, rotation=<rot 90 0 0>}]
 clavicle_l [{shapeType=Capsule, length=5, radius=7.5, shapeName=clavicle_l_capsule, center=<5.3519 0.599 -3.9969>, rotation=<rot 86 -179.999 175.001>}]
 clavicle_r [{shapeType=Capsule, length=5, radius=7.5, shapeName=clavicle_l_capsule, center=<-5.3519 -0.599 3.9969>, rotation=<rot 86 -179.999 175.001>}]
 dyn_C_backDress_1 [{shapeType=Capsule, length=21, radius=11, shapeName=dyn_C_backDress_1_capsule, center=<-16.79 0 -7.8661>, rotation=<rot 0 0 90>}]
 dyn_C_backDress_2 [{shapeType=Capsule, length=21, radius=11, shapeName=dyn_C_backDress_2_capsule, center=<-16.79 0 -7.8661>, rotation=<rot 0 0 90>}]
 dyn_C_backDress_3 [{shapeType=Capsule, length=21, radius=11, shapeName=dyn_C_backDress_3_capsule, center=<-16.79 0 -7.8661>, rotation=<rot 0 0 90>}]
 dyn_C_backDress_4 [{shapeType=Capsule, length=21, radius=11, shapeName=dyn_C_backDress_4_capsule, center=<-16.79 0 -7.8661>, rotation=<rot 0 0 90>}]
 dyn_L_backDress_1 [{shapeType=Capsule, length=10, radius=12.5, shapeName=dyn_L_backDress_1_capsule, center=<16.7961 -0.03 -8.4734>, rotation=<rot 90 0 0>}]
 dyn_L_backDress_2 [{shapeType=Capsule, length=10, radius=12.5, shapeName=dyn_L_backDress_2_capsule, center=<16.7961 -0.03 -8.4734>, rotation=<rot 90 0 0>}]
 dyn_L_backDress_3 [{shapeType=Capsule, length=10, radius=12.5, shapeName=dyn_L_backDress_3_capsule, center=<16.7961 -0.03 -8.4734>, rotation=<rot 90 0 0>}]
 dyn_L_backDress_4 [{shapeType=Capsule, length=10, radius=12.5, shapeName=dyn_L_backDress_4_capsule, center=<16.7961 -0.03 -8.4734>, rotation=<rot 90 0 0>}]
 dyn_L_sideDress_1 [{shapeType=Capsule, length=10, radius=11, shapeName=dyn_L_sideDress_1_capsule, center=<10.5266 0 -6.6469>, rotation=<rot 90 0 0>}]
 dyn_L_sideDress_2 [{shapeType=Capsule, length=10, radius=11, shapeName=dyn_L_sideDress_2_capsule, center=<10.5266 0 -6.6469>, rotation=<rot 90 0 0>}]
 dyn_L_sideDress_3 [{shapeType=Capsule, length=10, radius=11, shapeName=dyn_L_sideDress_3_capsule, center=<10.5266 0 -6.6469>, rotation=<rot 90 0 0>}]
 dyn_R_backDress_1 [{shapeType=Capsule, length=10, radius=12.5, shapeName=dyn_R_backDress_1_capsule, center=<-16.7961 0.03 8.4734>, rotation=<rot 90 0 0>}]
 dyn_R_backDress_2 [{shapeType=Capsule, length=10, radius=12.5, shapeName=dyn_R_backDress_2_capsule, center=<-16.7961 0.03 8.4734>, rotation=<rot 90 0 0>}]
 dyn_R_backDress_3 [{shapeType=Capsule, length=10, radius=12.5, shapeName=dyn_R_backDress_3_capsule, center=<-16.7961 0.03 8.4734>, rotation=<rot 90 0 0>}]
 dyn_R_backDress_4 [{shapeType=Capsule, length=10, radius=12.5, shapeName=dyn_R_backDress_4_capsule, center=<-16.7961 0.03 8.4734>, rotation=<rot 90 0 0>}]
 dyn_R_sideDress_1 [{shapeType=Capsule, length=10, radius=11, shapeName=dyn_R_sideDress_1_capsule, center=<-9.6278 0 6.6469>, rotation=<rot 90 0 0>}]
 dyn_R_sideDress_2 [{shapeType=Capsule, length=10, radius=11, shapeName=dyn_R_sideDress_2_capsule, center=<-9.6278 0 6.6469>, rotation=<rot 90 0 0>}]
 dyn_R_sideDress_3 [{shapeType=Capsule, length=10, radius=11, shapeName=dyn_R_sideDress_3_capsule, center=<-9.6278 0 6.6469>, rotation=<rot 90 0 0>}]
 foot_l [{shapeType=Capsule, length=10, radius=8, shapeName=foot_l_capsule, center=<-4.1783 5.5946 0.5086>, rotation=<rot 0 0 90>}]
 foot_r [{shapeType=Capsule, length=10, radius=8, shapeName=foot_r_capsule, center=<4.1783 -5.5946 -0.5086>, rotation=<rot 0 0 90>}]
 lowerarm_l [{shapeType=Capsule, length=20, radius=6, shapeName=lowerarm_l_capsule, center=<10 0 0>, rotation=<rot 90 0 0>}]
 lowerarm_r [{shapeType=Capsule, length=20, radius=6, shapeName=lowerarm_l_capsule, center=<-10 0 0>, rotation=<rot 90 0 0>}]
 pelvis [{
   shapeType=Capsule
   length=11.6533
   radius=9.77
   shapeName=pelvis_capsule
   center=<0.8194 -1.6022 -0.3109>
   rotation=<rot 0 -3.8946 -178.4815>
  }]
 spine_01 [{
   shapeType=Capsule
   length=13.196
   radius=9.7607
   shapeName=spine_01_capsule
   center=<-0.0825 0.3847 -0.0166>
   rotation=<rot -1.154 -90 0.0952>
  }]
 spine_02 [{
   shapeType=Capsule
   length=11.4453
   radius=10.0245
   shapeName=spine_02_capsule
   center=<-0.077 1.7345 0.0511>
   rotation=<rot 3.5501 -90 -0.234>
  }]
 spine_03 [{shapeType=Capsule, length=12.1167, radius=9.4506, shapeName=spine_03_capsule, center=<0 1.3074 0.0009>, rotation=<rot 0.0445 -90 0.0029>}]
 spine_04 [{shapeType=Capsule, length=12.1167, radius=9.4506, shapeName=spine_03_capsule, center=<0 0.4035 0.0016>, rotation=<rot 0.0445 -90 0.0029>}]
 spine_05 [{shapeType=Capsule, length=15.4242, radius=8.6742, shapeName=spine_05_capsule, center=<0 0.2421 0>}]
 thigh_l [{shapeType=Capsule, length=30, radius=9, shapeName=thigh_l_capsule, center=<-23.5249 0 0>, rotation=<rot 90 0 0>}]
 thigh_r [{shapeType=Capsule, length=30, radius=9, shapeName=thigh_r_capsule, center=<23.5249 0 0>, rotation=<rot 90 0 0>}]
 upperarm_l [{shapeType=Capsule, length=20, radius=5.5, shapeName=upperarm_l_capsule, center=<13 0 0>, rotation=<rot 90 0 0>}]
 upperarm_r [{shapeType=Capsule, length=20, radius=5.5, shapeName=upperarm_l_capsule, center=<-13 0 0>, rotation=<rot 90 0 0>}]
}
constraints []
